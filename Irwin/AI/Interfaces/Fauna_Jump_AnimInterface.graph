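# Animation layer interface for fauna jump animation, declared as an AnimBlueprint with no target skeleton.
# It defines a single layer, Fauna Jump, containing only an Output Pose with no input poses and no layer group.
# It holds no variables or logic, so the jump pose itself comes from whichever animation blueprint implements the layer.
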AnimBlueprint /Irwin/AI/Interfaces/Fauna_Jump_AnimInterface
target_skeleton=null
graphs [Fauna Jump {nodes [Output Pose() {layerGroup=None}]}]
variables []
