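# Animation Blueprint for Robert, a rideable predator dinosaur, covering both AI and player-ridden movement.
# AI locomotion (idle, moving, jumping, swimming, sleeping) or riding and boost state machines, selected by bool, feed a cached pose with additive layers.
# Linked Fauna idle and jump layers and montage slots add shared behavior and actions, smoothed by inertialization.
# A switchable final pass applies weighted head look-at, biped slope warping and curve-driven leg IK.
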AnimBlueprint /Irwin/AI/Predators/Robert/Animation/Robert_AnimBP
target_skeleton=/Irwin/AI/Base/Robert_Dinosaur_Invis_Skeleton
graphs [
 AnimGraph {nodes [
   n1 = Weighted Look At - Bone: BipedHeadNeck_A_Head_C!(
    ComponentPose:Component Space Pose Link Structure=Local To Component(LocalPose=Use cached pose 'SavedPose30689'())
    Alpha:float=Get LookAtAlpha() -> (LookAtAlpha:float)
    LookAtLocation:Vector=Get AimTargetLocation() -> (AimTargetLocation:Vector)
   ) -> (Pose:Component Space Pose Link Structure) {
    is=title FortAnimGraphNode_WeightedLookAt
    bUseLookUpAxis=true
    interpolationTime=3
    interpolationType=EaseInOutExponent3
    lookAtClamp=80
    alphaBoolBlend {blendOption=Linear}
    boneToModify {boneName=BipedHeadNeck_A_Head_C}
    boneWeightsToApply [{weight=0.25}, {weight=0.25}, {weight=0.25}, {weight=0.25}]
    lookAt_Axis {bInLocalSpace=false}
    lookUp_Axis {bInLocalSpace=false, axis {x=0, z=1}}
   }
   n2 = Biped Slope Warping!(ComponentPose:Component Space Pose Link Structure=@n1, AlphaCurveName:Name=BothLegIK) -> (Pose:Component Space Pose Link Structure) {
    is=title FortAnimGraphNode_SlopeWarping
    alphaInputType=Curve
    alphaBoolBlend {blendOption=Linear}
    alphaScaleBiasClamp {bias=1, scale=-1}
    feetDefinitions [
     {numBonesInLimb=3, fKFootBone {boneName=HoofedLeg_A_Ball_L}, iKFootBone {boneName=IK_Foot_L}}
     {numBonesInLimb=3, fKFootBone {boneName=HoofedLeg_A_Ball_R}, iKFootBone {boneName=IK_Foot_R}}
    ]
    iKFootRootBone {boneName=IK_Foot_Root}
    pelvisBone {boneName=QuadSpine_A_Pelvis_C}
   }
   n3 = Leg IK!(ComponentPose:Component Space Pose Link Structure=@n2, AlphaCurveName:Name=BothLegIK) -> (Pose:Component Space Pose Link Structure) {
    is=title AnimGraphNode_LegIK
    alphaCurveName=BothLegIK
    alphaInputType=Curve
    alphaBoolBlend {blendOption=Linear}
    alphaScaleBiasClamp {bias=1, scale=-1}
    legsDefinition [
     {
      bEnableKneeTwistCorrection=true
      bEnableRotationLimit=false
      footBoneForwardAxis=Y
      hingeRotationAxis=None
      minRotationAngle=15
      numBonesInLimb=3
      twistOffsetCurveName=None
      fKFootBone {boneName=HoofedLeg_A_Ball_L}
      iKFootBone {boneName=IK_Foot_L}
     }
     {
      bEnableKneeTwistCorrection=true
      bEnableRotationLimit=false
      footBoneForwardAxis=Y
      hingeRotationAxis=None
      minRotationAngle=15
      numBonesInLimb=3
      twistOffsetCurveName=None
      fKFootBone {boneName=HoofedLeg_A_Ball_R}
      iKFootBone {boneName=IK_Foot_R}
     }
    ]
   }
   n4 = Component To Local(ComponentPose=@n3)
   n5 = Use cached pose 'SavedPose30689'()
   n6 = Get bIsSlopeWarpHack() -> (bIsSlopeWarpHack:bool)
   Output Pose(Result=Slot 'SequencerDefault'(Source=Blend Poses by bool(BlendPose_0=@n5, BlendPose_1=@n4, BlendTime_1=0.25, bActiveValue=@n6) {blendType=Linear}) {slotName=SequencerDefault}) {layerGroup=None}
   AnimGraphNode_StateMachine_1 = AI_LocomotionPoseAdditive() {is=title AnimGraphNode_StateMachine}
   AnimGraphNode_StateMachine_4 = AI_Locomotion() {is=title AnimGraphNode_StateMachine}
   n7 = Apply Additive!(
    Base=Apply Additive!(
     Base=@AnimGraphNode_StateMachine_4
     Additive=@AnimGraphNode_StateMachine_1
     Alpha:float=Get LocomotionPoseAdditiveAlpha() -> (LocomotionPoseAdditiveAlpha:float)
    ) -> (Pose) {alphaBoolBlend {blendOption=Linear}}
    Additive="Sequence Player (additive) 'Robert_eyeAdditive'"() {sequence=&/Robert_eyeAdditive}
    AlphaCurveName:Name=DisableFacialAdditive
   ) -> (Pose) {alphaInputType=Curve, alphaBoolBlend {blendOption=Linear}, alphaScaleBiasClamp {bClampResult=true, bias=1, scale=-1}}
   AnimGraphNode_StateMachine_3 = RiddenLocomotionAdditive() {is=title AnimGraphNode_StateMachine, bAllowConduitEntryStates=true}
   AnimGraphNode_StateMachine_5 = RobertRidingStateMachine() {is=title AnimGraphNode_StateMachine, bAllowConduitEntryStates=true}
   n8 = Apply Additive(
    Base=@AnimGraphNode_StateMachine_5
    Additive=Inertialization!(Source:Pose Link Structure=@AnimGraphNode_StateMachine_3) -> (Pose:Pose Link Structure) {is=title AnimGraphNode_Inertialization}
   ) {alphaBoolBlend {blendOption=Linear}}
   AnimGraphNode_StateMachine_6 = BoostStatemachine() {is=title AnimGraphNode_StateMachine}
   n9 = Get bIsAllowBoost() -> (bIsAllowBoost:bool)
   Save cached pose 'LocomotionPose'(Pose=Blend Poses by bool(
     BlendPose_0=Blend Poses by bool(BlendPose_0=@AnimGraphNode_StateMachine_6, BlendPose_1=@n8, BlendTime_0=0.5, BlendTime_1=0.5, bActiveValue=@n9) {blendType=Linear, transitionType=Inertialization}
     BlendPose_1=@n7
     BlendTime_0=0.2
     BlendTime_1=0.4
     bActiveValue=Get bIsRidden() -> (bIsRidden:bool)
    ) {blendType=Linear, transitionType=Inertialization}) {cacheName=LocomotionPose}
   Save cached pose 'FaunaIdlePose'(Pose=Fauna_Idle_AnimInterface - Fauna Idle!(InPose Idle:Pose Link Structure) -> (Pose:Pose Link Structure) {is=title AnimGraphNode_LinkedAnimLayer, interface=/Irwin/AI/Interfaces/Fauna_Idle_AnimInterface.Fauna_Idle_AnimInterface_C, layer=Fauna Idle}) {cacheName=FaunaIdlePose}
   n10 = Layered blend per bone(
    BasePose=Use cached pose 'LocomotionPose'()
    BlendPoses_0=Slot 'UpperBodyInMotion'(Source=Use cached pose 'LocomotionPose'()) {slotName=UpperBodyInMotion}
   ) {
    bMeshSpaceRotationBlend=true
    layerSetup [{branchFilters [{blendDepth=5, boneName=C_QuadSpine_A_Spine1_Jnt}, {blendDepth=2, boneName=IK_Foot_R}, {blendDepth=2, boneName=IK_Foot_L}]}]
   }
   n11 = Slot 'FullBodyInPlace'(Source=Use cached pose 'LocomotionPose'()) {slotName=FullBodyInPlace}
   n12 = Get bIsFullBodyInPlace() -> (bIsFullBodyInPlace:bool)
   n13 = Fauna_Jump_AnimInterface - Fauna Jump!() -> (Pose:Pose Link Structure) {
    is=title AnimGraphNode_LinkedAnimLayer
    bPropagateNotifiesToLinkedInstances=true
    bReceiveNotifiesFromLinkedInstances=true
    interface=/Irwin/AI/Interfaces/Fauna_Jump_AnimInterface.Fauna_Jump_AnimInterface_C
    layer=Fauna Jump
   }
   n14 = Slot 'DefaultSlot'(Source=Blend Poses by bool(
     BlendPose_0=@n13
     BlendPose_1=Blend Poses by bool(BlendPose_0=@n11, BlendPose_1=@n10, BlendTime_0=0.4, BlendTime_1=0.4, bActiveValue=@n12) {blendType=Linear, transitionType=Inertialization}
     BlendTime_0=0.4
     BlendTime_1=0.2
     bActiveValue=Get bIsFloating() -> (bIsFloating:bool)
    ) {blendType=Linear, transitionType=Inertialization})
   Save cached pose 'SavedPose30689'(Pose=Inertialization(Source=Slot 'FullBodyAdditiveSlot'(Source=@n14) {slotName=FullBodyAdditiveSlot})) {cacheName=SavedPose30689}
  ]}
 AI_LocomotionPoseAdditive {
  entry=Default
  path=AnimGraph.AnimGraphNode_StateMachine_1.AI_LocomotionPoseAdditive
  ids=id
  nodes [
   AnimStateNode_0 = Default() {is=title AnimStateNode}
   AnimStateNode_1 = HeadTiltDownPose() {is=title AnimStateNode}
   AnimStateTransitionNode_0 = AnimStateTransitionNode() {crossfadeDuration=0.4, logicType=TLT_Inertialization}
   AnimStateTransitionNode_1 = AnimStateTransitionNode() {bAutomaticRuleBasedOnSequencePlayerInState=true, crossfadeDuration=0.5, logicType=TLT_Inertialization}
  ]
  links [
   {
    dst=HeadTiltDownPose
    rule=AnimGraph.AnimGraphNode_StateMachine_1.AI_LocomotionPoseAdditive.AnimStateTransitionNode_0.Transition
    src=Default
    via=AnimStateTransitionNode_0
   }
   {
    dst=Default
    rule=AnimGraph.AnimGraphNode_StateMachine_1.AI_LocomotionPoseAdditive.AnimStateTransitionNode_1.Transition
    src=HeadTiltDownPose
    via=AnimStateTransitionNode_1
   }
  ]
 }
 Default {
  path=AnimGraph.AnimGraphNode_StateMachine_1.AI_LocomotionPoseAdditive.AnimStateNode_0.Default
  nodes [Output Animation Pose(Result=Additive Identity Pose())]
 }
 HeadTiltDownPose {
  path=AnimGraph.AnimGraphNode_StateMachine_1.AI_LocomotionPoseAdditive.AnimStateNode_1.HeadTiltDownPose
  nodes [Output Animation Pose(Result="Sequence Player (additive) 'Robert_LocomotionAdditive_TiltDown_Pose'"() {sequence=&/Robert_LocomotionAdditive_TiltDown_Pose})]
 }
 Transition {
  path=AnimGraph.AnimGraphNode_StateMachine_1.AI_LocomotionPoseAdditive.AnimStateTransitionNode_0.Transition
  nodes [Result(bCanEnterTransition=Get bEnableHeadTiltDownAdditive() -> (bEnableHeadTiltDownAdditive:bool))]
 }
 Transition {path=AnimGraph.AnimGraphNode_StateMachine_1.AI_LocomotionPoseAdditive.AnimStateTransitionNode_1.Transition, nodes [Result()]}
 RiddenLocomotionAdditive {
  entry=Conduit
  path=AnimGraph.AnimGraphNode_StateMachine_3.RiddenLocomotionAdditive
  ids=id
  nodes [
   AnimStateConduitNode_0 = Conduit()
   AnimStateNode_0 = AdditivePose() {is=title AnimStateNode}
   AnimStateNode_1 = Jump() {is=title AnimStateNode}
   AnimStateNode_2 = Land() {is=title AnimStateNode}
   AnimStateNode_4 = Entry() {is=title AnimStateNode}
   AnimStateTransitionNode_0 = AnimStateTransitionNode()
   AnimStateTransitionNode_1 = AnimStateTransitionNode() {logicType=TLT_Inertialization}
   AnimStateTransitionNode_12 = AnimStateTransitionNode() {crossfadeDuration=0.4, logicType=TLT_Inertialization, priorityOrder=100}
   AnimStateTransitionNode_14 = AnimStateTransitionNode()
   AnimStateTransitionNode_15 = AnimStateTransitionNode() {logicType=TLT_Inertialization}
   AnimStateTransitionNode_16 = AnimStateTransitionNode()
   AnimStateTransitionNode_2 = AnimStateTransitionNode() {crossfadeDuration=0.1, logicType=TLT_Inertialization, priorityOrder=100}
   AnimStateTransitionNode_4 = AnimStateTransitionNode() {logicType=TLT_Inertialization, priorityOrder=2}
   AnimStateTransitionNode_5 = AnimStateTransitionNode() {bAutomaticRuleBasedOnSequencePlayerInState=true}
  ]
  links [
   {
    dst=Conduit
    rule=AnimGraph.AnimGraphNode_StateMachine_3.RiddenLocomotionAdditive.AnimStateTransitionNode_14.Transition
    src=AdditivePose
    via=AnimStateTransitionNode_14
   }
   {
    dst=AdditivePose
    rule=AnimGraph.AnimGraphNode_StateMachine_3.RiddenLocomotionAdditive.AnimStateTransitionNode_12.Transition
    src=Conduit
    via=AnimStateTransitionNode_12
   }
   {
    dst=Entry
    rule=AnimGraph.AnimGraphNode_StateMachine_3.RiddenLocomotionAdditive.AnimStateTransitionNode_4.Transition
    src=Conduit
    via=AnimStateTransitionNode_4
   }
   {
    dst=Jump
    rule=AnimGraph.AnimGraphNode_StateMachine_3.RiddenLocomotionAdditive.AnimStateTransitionNode_1.Transition
    src=Conduit
    via=AnimStateTransitionNode_1
   }
   {
    dst=Conduit
    rule=AnimGraph.AnimGraphNode_StateMachine_3.RiddenLocomotionAdditive.AnimStateTransitionNode_5.Transition
    src=Entry
    via=AnimStateTransitionNode_5
   }
   {
    dst=Conduit
    rule=AnimGraph.AnimGraphNode_StateMachine_3.RiddenLocomotionAdditive.AnimStateTransitionNode_0.Transition
    src=Jump
    via=AnimStateTransitionNode_0
   }
   {
    dst=Land
    rule=AnimGraph.AnimGraphNode_StateMachine_3.RiddenLocomotionAdditive.AnimStateTransitionNode_15.Transition
    src=Jump
    via=AnimStateTransitionNode_15
   }
   {
    dst=Conduit
    rule=AnimGraph.AnimGraphNode_StateMachine_3.RiddenLocomotionAdditive.AnimStateTransitionNode_16.Transition
    src=Land
    via=AnimStateTransitionNode_16
   }
   {
    dst=Jump
    rule=AnimGraph.AnimGraphNode_StateMachine_3.RiddenLocomotionAdditive.AnimStateTransitionNode_2.Transition
    src=Land
    via=AnimStateTransitionNode_2
   }
  ]
 }
 Conduit {
  path=AnimGraph.AnimGraphNode_StateMachine_3.RiddenLocomotionAdditive.AnimStateConduitNode_0.Conduit
  nodes [Result(bCanEnterTransition=true)]
 }
 AdditivePose {
  path=AnimGraph.AnimGraphNode_StateMachine_3.RiddenLocomotionAdditive.AnimStateNode_0.AdditivePose
  nodes [Output Animation Pose(Result=Additive Identity Pose())]
 }
 Jump {
  path=AnimGraph.AnimGraphNode_StateMachine_3.RiddenLocomotionAdditive.AnimStateNode_1.Jump
  nodes [Output Animation Pose(Result=Additive Identity Pose())]
 }
 Land {
  path=AnimGraph.AnimGraphNode_StateMachine_3.RiddenLocomotionAdditive.AnimStateNode_2.Land
  nodes [Output Animation Pose(Result="Sequence Player (additive) 'Robert_Riding_LandInMotion'"() {bLoopAnimation=false, sequence=&/Robert_Riding_LandInMotion})]
 }
 Entry {
  path=AnimGraph.AnimGraphNode_StateMachine_3.RiddenLocomotionAdditive.AnimStateNode_4.Entry
  nodes [Output Animation Pose(Result="Sequence Player (additive) 'Robert_Riding_Entry'"() {bLoopAnimation=false, sequence=&/Robert_Riding_Entry})]
 }
 Transition {
  path=AnimGraph.AnimGraphNode_StateMachine_3.RiddenLocomotionAdditive.AnimStateTransitionNode_0.Transition
  nodes [Result(bCanEnterTransition=NOT Boolean(A=Get bIsAccelerating2D()))]
 }
 Transition {
  path=AnimGraph.AnimGraphNode_StateMachine_3.RiddenLocomotionAdditive.AnimStateTransitionNode_1.Transition
  nodes [Result(bCanEnterTransition=Get bStateRule_IsJumpAdditive() -> (bStateRule_IsJumpAdditive:bool))]
 }
 Transition {
  path=AnimGraph.AnimGraphNode_StateMachine_3.RiddenLocomotionAdditive.AnimStateTransitionNode_12.Transition
  nodes [Result(bCanEnterTransition=true)]
 }
 Transition {
  path=AnimGraph.AnimGraphNode_StateMachine_3.RiddenLocomotionAdditive.AnimStateTransitionNode_14.Transition
  nodes [Result(bCanEnterTransition=true)]
 }
 Transition {
  path=AnimGraph.AnimGraphNode_StateMachine_3.RiddenLocomotionAdditive.AnimStateTransitionNode_15.Transition
  nodes [Result(bCanEnterTransition=Get bStateRule_IsJumpToJumpLand() -> (bStateRule_IsJumpToJumpLand:bool))]
 }
 Transition {
  path=AnimGraph.AnimGraphNode_StateMachine_3.RiddenLocomotionAdditive.AnimStateTransitionNode_16.Transition
  nodes [Result(bCanEnterTransition=Get bStateRule_IsAdditiveInterruptible() -> (bStateRule_IsAdditiveInterruptible:bool))]
 }
 Transition {
  path=AnimGraph.AnimGraphNode_StateMachine_3.RiddenLocomotionAdditive.AnimStateTransitionNode_2.Transition
  nodes [Result(bCanEnterTransition=Get bStateRule_IsLandAdditiveToJump() -> (bStateRule_IsLandAdditiveToJump:bool))]
 }
 Transition {
  path=AnimGraph.AnimGraphNode_StateMachine_3.RiddenLocomotionAdditive.AnimStateTransitionNode_4.Transition
  nodes [Result(bCanEnterTransition=Get bJustGotPlayer() -> (bJustGotPlayer:bool))]
 }
 Transition {path=AnimGraph.AnimGraphNode_StateMachine_3.RiddenLocomotionAdditive.AnimStateTransitionNode_5.Transition, nodes [Result()]}
 AI_Locomotion {
  entry=Idle
  path=AnimGraph.AnimGraphNode_StateMachine_4.AI_Locomotion
  ids=id
  nodes [
   IsSwimming() {is=title AnimStateAliasNode, also {aliases [JumpStart,Land]}}
   AnimStateNode_0 = Idle() {stateLeft {guid=#1, notifyName=IdleExit}}
   AnimStateNode_10 = Moving() {is=title AnimStateNode}
   AnimStateNode_2 = JumpStart() {is=title AnimStateNode}
   AnimStateNode_3 = FallAsleep() {is=title AnimStateNode}
   AnimStateNode_4 = Sleeping() {is=title AnimStateNode}
   AnimStateNode_5 = Fall() {is=title AnimStateNode, stateLeft {guid=#2, notifyName=FallExit}}
   AnimStateNode_6 = Land() {stateEntered {guid=#3, notifyName=LandEnter}}
   AnimStateNode_7 = WakeUp() {is=title AnimStateNode}
   AnimStateNode_9 = Swimming() {is=title AnimStateNode, stateEntered {guid=#4, notifyName=SwimmingEnter}}
   AnimStateTransitionNode_0 = AnimStateTransitionNode() {crossfadeDuration=0.3, logicType=TLT_Inertialization}
   AnimStateTransitionNode_1 = AnimStateTransitionNode() {logicType=TLT_Inertialization}
   AnimStateTransitionNode_10 = AnimStateTransitionNode()
   AnimStateTransitionNode_2 = AnimStateTransitionNode()
   AnimStateTransitionNode_28 = AnimStateTransitionNode() {crossfadeDuration=0.6, logicType=TLT_Inertialization}
   AnimStateTransitionNode_3 = AnimStateTransitionNode() {crossfadeDuration=0.1, logicType=TLT_Inertialization, priorityOrder=2}
   AnimStateTransitionNode_31 = AnimStateTransitionNode() {crossfadeDuration=0.4, logicType=TLT_Inertialization}
   AnimStateTransitionNode_36 = AnimStateTransitionNode() {crossfadeDuration=0.4, logicType=TLT_Inertialization, priorityOrder=2}
   AnimStateTransitionNode_37 = AnimStateTransitionNode() {crossfadeDuration=0.4, logicType=TLT_Inertialization, priorityOrder=3}
   AnimStateTransitionNode_39 = AnimStateTransitionNode() {crossfadeDuration=0.4, logicType=TLT_Inertialization}
   AnimStateTransitionNode_4 = AnimStateTransitionNode() {priorityOrder=100}
   AnimStateTransitionNode_41 = AnimStateTransitionNode() {crossfadeDuration=0.4, logicType=TLT_Inertialization, priorityOrder=100}
   AnimStateTransitionNode_43 = AnimStateTransitionNode() {crossfadeDuration=0.1, logicType=TLT_Inertialization, priorityOrder=2}
   AnimStateTransitionNode_44 = AnimStateTransitionNode() {bAutomaticRuleBasedOnSequencePlayerInState=true, crossfadeDuration=0.35, logicType=TLT_Inertialization, priorityOrder=100}
   AnimStateTransitionNode_45 = AnimStateTransitionNode() {crossfadeDuration=0.1, logicType=TLT_Inertialization, priorityOrder=100}
   AnimStateTransitionNode_47 = AnimStateTransitionNode() {crossfadeDuration=0.1, logicType=TLT_Inertialization, priorityOrder=2}
   AnimStateTransitionNode_48 = AnimStateTransitionNode() {crossfadeDuration=0.45, logicType=TLT_Inertialization, priorityOrder=100}
   AnimStateTransitionNode_49 = AnimStateTransitionNode() {crossfadeDuration=0.4, logicType=TLT_Inertialization, priorityOrder=2}
   AnimStateTransitionNode_5 = AnimStateTransitionNode()
   AnimStateTransitionNode_6 = AnimStateTransitionNode()
   AnimStateTransitionNode_7 = AnimStateTransitionNode() {logicType=TLT_Inertialization}
   AnimStateTransitionNode_8 = AnimStateTransitionNode() {bAutomaticRuleBasedOnSequencePlayerInState=true, crossfadeDuration=0.6, logicType=TLT_Inertialization}
  ]
  links [
   {
    dst=Land
    rule=AnimGraph.AnimGraphNode_StateMachine_4.AI_Locomotion.AnimStateTransitionNode_43.Transition
    src=Fall
    via=AnimStateTransitionNode_43
   }
   {
    dst=Moving
    rule=AnimGraph.AnimGraphNode_StateMachine_4.AI_Locomotion.AnimStateTransitionNode_1.Transition
    src=Fall
    via=AnimStateTransitionNode_1
   }
   {
    dst=Swimming
    rule=AnimGraph.AnimGraphNode_StateMachine_4.AI_Locomotion.AnimStateTransitionNode_48.Transition
    src=Fall
    via=AnimStateTransitionNode_48
   }
   {
    dst=Sleeping
    rule=AnimGraph.AnimGraphNode_StateMachine_4.AI_Locomotion.AnimStateTransitionNode_5.AnimationTransitionGraph_0
    src=FallAsleep
    via=AnimStateTransitionNode_5
   }
   {
    dst=FallAsleep
    rule=AnimGraph.AnimGraphNode_StateMachine_4.AI_Locomotion.AnimStateTransitionNode_2.Transition
    src=Idle
    via=AnimStateTransitionNode_2
   }
   {
    dst=JumpStart
    rule=AnimGraph.AnimGraphNode_StateMachine_4.AI_Locomotion.AnimStateTransitionNode_45.Transition
    src=Idle
    via=AnimStateTransitionNode_45
   }
   {
    dst=Moving
    rule=AnimGraph.AnimGraphNode_StateMachine_4.AI_Locomotion.AnimStateTransitionNode_28.Transition
    src=Idle
    via=AnimStateTransitionNode_28
   }
   {
    dst=Swimming
    rule=AnimGraph.AnimGraphNode_StateMachine_4.AI_Locomotion.AnimStateTransitionNode_36.Transition
    src=Idle
    via=AnimStateTransitionNode_36
   }
   {
    dst=Swimming
    rule=AnimGraph.AnimGraphNode_StateMachine_4.AI_Locomotion.AnimStateTransitionNode_0.Transition
    src=IsSwimming
    via=AnimStateTransitionNode_0
   }
   {
    dst=Fall
    rule=AnimGraph.AnimGraphNode_StateMachine_4.AI_Locomotion.AnimStateTransitionNode_8.Transition
    src=JumpStart
    via=AnimStateTransitionNode_8
   }
   {
    dst=Idle
    rule=AnimGraph.AnimGraphNode_StateMachine_4.AI_Locomotion.AnimStateTransitionNode_44.Transition
    src=Land
    via=AnimStateTransitionNode_44
   }
   {
    dst=JumpStart
    rule=AnimGraph.AnimGraphNode_StateMachine_4.AI_Locomotion.AnimStateTransitionNode_47.Transition
    src=Land
    via=AnimStateTransitionNode_47
   }
   {
    dst=Moving
    rule=AnimGraph.AnimGraphNode_StateMachine_4.AI_Locomotion.AnimStateTransitionNode_7.Transition
    src=Land
    via=AnimStateTransitionNode_7
   }
   {
    dst=Fall
    rule=AnimGraph.AnimGraphNode_StateMachine_4.AI_Locomotion.AnimStateTransitionNode_41.Transition
    src=Moving
    via=AnimStateTransitionNode_41
   }
   {
    dst=Idle
    rule=AnimGraph.AnimGraphNode_StateMachine_4.AI_Locomotion.AnimStateTransitionNode_31.Transition
    src=Moving
    via=AnimStateTransitionNode_31
   }
   {
    dst=JumpStart
    rule=AnimGraph.AnimGraphNode_StateMachine_4.AI_Locomotion.AnimStateTransitionNode_3.Transition
    src=Moving
    via=AnimStateTransitionNode_3
   }
   {
    dst=Swimming
    rule=AnimGraph.AnimGraphNode_StateMachine_4.AI_Locomotion.AnimStateTransitionNode_37.Transition
    src=Moving
    via=AnimStateTransitionNode_37
   }
   {
    dst=WakeUp
    rule=AnimGraph.AnimGraphNode_StateMachine_4.AI_Locomotion.AnimStateTransitionNode_10.AnimationTransitionGraph_0
    src=Sleeping
    via=AnimStateTransitionNode_10
   }
   {
    dst=Fall
    rule=AnimGraph.AnimGraphNode_StateMachine_4.AI_Locomotion.AnimStateTransitionNode_49.Transition
    src=Swimming
    via=AnimStateTransitionNode_49
   }
   {
    dst=JumpStart
    rule=AnimGraph.AnimGraphNode_StateMachine_4.AI_Locomotion.AnimStateTransitionNode_4.Transition
    src=Swimming
    via=AnimStateTransitionNode_4
   }
   {
    dst=Moving
    rule=AnimGraph.AnimGraphNode_StateMachine_4.AI_Locomotion.AnimStateTransitionNode_39.Transition
    src=Swimming
    via=AnimStateTransitionNode_39
   }
   {
    dst=Idle
    rule=AnimGraph.AnimGraphNode_StateMachine_4.AI_Locomotion.AnimStateTransitionNode_6.Transition
    src=WakeUp
    via=AnimStateTransitionNode_6
   }
  ]
 }
 Idle {
  path=AnimGraph.AnimGraphNode_StateMachine_4.AI_Locomotion.AnimStateNode_0.Idle
  nodes [
   n1 = Blendspace Player 'Robert_Walk_N_BS_TurnPose'!(X=Get IdleShuffleBodyBend() -> (IdleShuffleBodyBend:float)) -> (Pose) {blendSpace=&/Robert_Walk_N_BS_TurnPose}
   n2 = Get IdleShufflePlayRate() -> (IdleShufflePlayRate:float)
   n3 = Layered blend per bone(
    BasePose=Use cached pose 'FaunaIdlePose'()
    BlendPoses_0=Blend Poses by bool(
     BlendPose_0=Sequence Player 'Robert_TurnInPlace_R_90_HipsAndFeet'!(PlayRate:float=@n2) -> (Pose) {groupName=IdleTurn, method=SyncGroup, sequence=&/Robert_TurnInPlace_R_90_HipsAndFeet}
     BlendPose_1=Sequence Player 'Robert_TurnInPlace_L_90_HipsAndFeet'!(PlayRate:float=@n2) -> (Pose) {groupName=IdleTurn, method=SyncGroup, sequence=&/Robert_TurnInPlace_L_90_HipsAndFeet}
     BlendTime_0=0.5
     BlendTime_1=0.5
     bActiveValue=Get bIdleShuffleTurnDirection() -> (bIdleShuffleTurnDirection:bool)
    ) {blendType=Linear}
   ) {layerSetup [{branchFilters [{boneName=QuadSpine_A_Pelvis_C}, {blendDepth=-1, boneName=Tail_A_TailBase_C}, {blendDepth=-1, boneName=QuadSpine_A_Spine1_C}]}]}
   n4 = Get bLookAtAdditiveTurnAlpha() -> (bLookAtAdditiveTurnAlpha:bool)
   Output Animation Pose(Result=Blend Poses by bool(
     BlendPose_0=Apply Additive!(Base=@n3, Additive=@n1, bAlphaBoolEnabled:bool=@n4) -> (Pose) {alphaInputType=Bool, alphaBoolBlend {blendInTime=0.4, blendOption=Linear, blendOutTime=0.4}}
     BlendPose_1=Use cached pose 'FaunaIdlePose'()
     BlendTime_0=0.6
     BlendTime_1=0.6
     bActiveValue=Get bIsIdleShuffleTurn() -> (bIsIdleShuffleTurn:bool)
    ) {blendType=Linear})
  ]
 }
 Moving {
  path=AnimGraph.AnimGraphNode_StateMachine_4.AI_Locomotion.AnimStateNode_10.Moving
  nodes [
   n1 = Blendspace Player 'Robert_WalkRun_BS'!(X=Get Speed() -> (Speed:float)) -> (Pose) {blendSpace=&/Robert_WalkRun_BS, groupName=Locomotion, method=SyncGroup}
   n2 = Blendspace Player 'Robert_Walk_N_BS_TurnPose'(X=BreakRotator(InRot=Get PawnRotationRate() -> (PawnRotationRate:Rotator)).Yaw) {blendSpace=&Robert_Walk_N_BS_TurnPose}
   n3 = Get bLookAtAdditiveTurnAlpha()
   AnimGraphNode_StateMachine_0 = JumpLand() {is=title AnimGraphNode_StateMachine}
   Output Animation Pose(Result=Apply Additive!(
     Base=Apply Additive!(Base=@n1, Additive=@n2, bAlphaBoolEnabled:bool=@n3) -> (Pose) {alphaInputType=Bool, alphaBoolBlend {blendInTime=0.4, blendOption=Linear, blendOutTime=0.4}}
     Additive=@AnimGraphNode_StateMachine_0
     Alpha:float
    ) -> (Pose) {alphaBoolBlend {blendOption=Linear}})
  ]
 }
 JumpLand {
  entry=AddRef
  path=AnimGraph.AnimGraphNode_StateMachine_4.AI_Locomotion.AnimStateNode_10.Moving.AnimGraphNode_StateMachine_0.JumpLand
  ids=id
  nodes [
   AnimStateNode_0 = AddRef() {is=title AnimStateNode}
   AnimStateNode_1 = JumpLandAdd() {is=title AnimStateNode, stateLeft {guid=#5, notifyName=LandAddExit}}
   AnimStateNode_2 = IdleToMoveAdd() {is=title AnimStateNode, stateLeft {guid=#6, notifyName=IdleToMoveExit}}
   AnimStateTransitionNode_0 = AnimStateTransitionNode() {logicType=TLT_Inertialization}
   AnimStateTransitionNode_1 = AnimStateTransitionNode() {bAutomaticRuleBasedOnSequencePlayerInState=true, crossfadeDuration=0.4, logicType=TLT_Inertialization}
   AnimStateTransitionNode_2 = AnimStateTransitionNode() {crossfadeDuration=0.1, priorityOrder=100}
   AnimStateTransitionNode_3 = AnimStateTransitionNode() {bAutomaticRuleBasedOnSequencePlayerInState=true, crossfadeDuration=0.4, logicType=TLT_Inertialization}
  ]
  links [
   {
    dst=IdleToMoveAdd
    rule=AnimGraph.AnimGraphNode_StateMachine_4.AI_Locomotion.AnimStateNode_10.Moving.AnimGraphNode_StateMachine_0.JumpLand.AnimStateTransitionNode_2.Transition
    src=AddRef
    via=AnimStateTransitionNode_2
   }
   {
    dst=JumpLandAdd
    rule=AnimGraph.AnimGraphNode_StateMachine_4.AI_Locomotion.AnimStateNode_10.Moving.AnimGraphNode_StateMachine_0.JumpLand.AnimStateTransitionNode_0.Transition
    src=AddRef
    via=AnimStateTransitionNode_0
   }
   {
    dst=AddRef
    rule=AnimGraph.AnimGraphNode_StateMachine_4.AI_Locomotion.AnimStateNode_10.Moving.AnimGraphNode_StateMachine_0.JumpLand.AnimStateTransitionNode_3.Transition
    src=IdleToMoveAdd
    via=AnimStateTransitionNode_3
   }
   {
    dst=AddRef
    rule=AnimGraph.AnimGraphNode_StateMachine_4.AI_Locomotion.AnimStateNode_10.Moving.AnimGraphNode_StateMachine_0.JumpLand.AnimStateTransitionNode_1.Transition
    src=JumpLandAdd
    via=AnimStateTransitionNode_1
   }
  ]
 }
 AddRef {
  path=AnimGraph.AnimGraphNode_StateMachine_4.AI_Locomotion.AnimStateNode_10.Moving.AnimGraphNode_StateMachine_0.JumpLand.AnimStateNode_0.AddRef
  nodes [Output Animation Pose(Result=Additive Identity Pose())]
 }
 JumpLandAdd {
  path=AnimGraph.AnimGraphNode_StateMachine_4.AI_Locomotion.AnimStateNode_10.Moving.AnimGraphNode_StateMachine_0.JumpLand.AnimStateNode_1.JumpLandAdd
  nodes [Output Animation Pose(Result="Sequence Player (additive) 'Robert_JumpLand_Moving_Add'"() {sequence=&/Robert_JumpLand_Moving_Add})]
 }
 IdleToMoveAdd {
  path=AnimGraph.AnimGraphNode_StateMachine_4.AI_Locomotion.AnimStateNode_10.Moving.AnimGraphNode_StateMachine_0.JumpLand.AnimStateNode_2.IdleToMoveAdd
  nodes [Output Animation Pose(Result="Sequence Player (additive) 'Robert_Idle_To_Moving_ADD'"() {sequence=&/Robert_Idle_To_Moving_ADD})]
 }
 Transition {
  path=AnimGraph.AnimGraphNode_StateMachine_4.AI_Locomotion.AnimStateNode_10.Moving.AnimGraphNode_StateMachine_0.JumpLand.AnimStateTransitionNode_0.Transition
  nodes [Result(bCanEnterTransition=Get bStateRule_PlayLandAdditive() -> (bStateRule_PlayLandAdditive:bool))]
 }
 Transition {
  path=AnimGraph.AnimGraphNode_StateMachine_4.AI_Locomotion.AnimStateNode_10.Moving.AnimGraphNode_StateMachine_0.JumpLand.AnimStateTransitionNode_1.Transition
  nodes [Result()]
 }
 Transition {
  path=AnimGraph.AnimGraphNode_StateMachine_4.AI_Locomotion.AnimStateNode_10.Moving.AnimGraphNode_StateMachine_0.JumpLand.AnimStateTransitionNode_2.Transition
  nodes [Result(bCanEnterTransition=Get bStateRule_PlayIdleToMoveAdditive() -> (bStateRule_PlayIdleToMoveAdditive:bool))]
 }
 Transition {
  path=AnimGraph.AnimGraphNode_StateMachine_4.AI_Locomotion.AnimStateNode_10.Moving.AnimGraphNode_StateMachine_0.JumpLand.AnimStateTransitionNode_3.Transition
  nodes [Result()]
 }
 JumpStart {
  path=AnimGraph.AnimGraphNode_StateMachine_4.AI_Locomotion.AnimStateNode_2.JumpStart
  nodes [Output Animation Pose(Result=Sequence Player 'Robert_Jump_Start'() {bLoopAnimation=false, sequence=&/Robert_Jump_Start})]
 }
 FallAsleep {
  path=AnimGraph.AnimGraphNode_StateMachine_4.AI_Locomotion.AnimStateNode_3.FallAsleep
  nodes [Output Animation Pose(Result=Sequence Player 'Robert_Riding_Exhausted_Intro'() {bLoopAnimation=false, sequence=&/Robert_Riding_Exhausted_Intro})]
 }
 Sleeping {
  path=AnimGraph.AnimGraphNode_StateMachine_4.AI_Locomotion.AnimStateNode_4.Sleeping
  nodes [Output Animation Pose(Result=Sequence Player 'Robert_Riding_Exhausted_Loop'() {sequence=&/Robert_Riding_Exhausted_Loop})]
 }
 Fall {
  path=AnimGraph.AnimGraphNode_StateMachine_4.AI_Locomotion.AnimStateNode_5.Fall
  nodes [Output Animation Pose(Result=Sequence Player 'Robert_Jump_FallLoop'() {sequence=&/Robert_Jump_FallLoop})]
 }
 Land {
  path=AnimGraph.AnimGraphNode_StateMachine_4.AI_Locomotion.AnimStateNode_6.Land
  nodes [Output Animation Pose(Result=Sequence Player 'Robert_Jump_Land'() {bLoopAnimation=false, sequence=&/Robert_Jump_Land})]
 }
 WakeUp {
  path=AnimGraph.AnimGraphNode_StateMachine_4.AI_Locomotion.AnimStateNode_7.WakeUp
  nodes [Output Animation Pose(Result=Sequence Player 'Robert_Riding_Exhausted_Recovery'() {bLoopAnimation=false, sequence=&/Robert_Riding_Exhausted_Recovery})]
 }
 Swimming {
  path=AnimGraph.AnimGraphNode_StateMachine_4.AI_Locomotion.AnimStateNode_9.Swimming
  nodes [
   n1 = Sequence Player 'Robert_Swim_Idle'() {sequence=&/Robert_Swim_Idle}
   n2 = Sequence Player 'Robert_Swim_N'() {sequence=&/Robert_Swim_N}
   n3 = Get bStateRule_SwimIdle_SwimLoco() -> (bStateRule_SwimIdle_SwimLoco:bool)
   n4 = BreakRotator(InRot=Get PawnRotationRate())
   Output Animation Pose(Result=Apply Additive!(
     Base=Blend Poses by bool(BlendPose_0=@n2, BlendPose_1=@n1, BlendTime_0=0.5, BlendTime_1=0.5, bActiveValue=@n3) {blendType=Linear, transitionType=Inertialization}
     Additive=Blendspace Player 'Robert_Walk_N_BS_TurnPose'(X=@n4.Yaw) {blendSpace=&Robert_Walk_N_BS_TurnPose}
     Alpha:float
    ) -> (Pose) {alphaBoolBlend {blendOption=Linear}})
  ]
 }
 Transition {
  path=AnimGraph.AnimGraphNode_StateMachine_4.AI_Locomotion.AnimStateTransitionNode_0.Transition
  nodes [Result(bCanEnterTransition=Get bIsSurfaceSwimming())]
 }
 Transition {
  path=AnimGraph.AnimGraphNode_StateMachine_4.AI_Locomotion.AnimStateTransitionNode_1.Transition
  nodes [Result(bCanEnterTransition=Get bStateRule_Fall_Moving() -> (bStateRule_Fall_Moving:bool))]
 }
 AnimationTransitionGraph_0 {
  path=AnimGraph.AnimGraphNode_StateMachine_4.AI_Locomotion.AnimStateTransitionNode_10.AnimationTransitionGraph_0
  nodes [Result(bCanEnterTransition=Get StateTransition_SleepingToWakeUp() -> (StateTransition_SleepingToWakeUp:bool))]
 }
 Transition {
  path=AnimGraph.AnimGraphNode_StateMachine_4.AI_Locomotion.AnimStateTransitionNode_2.Transition
  nodes [Result(bCanEnterTransition=Get StateTransition_IdleToFallAsleep() -> (StateTransition_IdleToFallAsleep:bool))]
 }
 Transition {
  path=AnimGraph.AnimGraphNode_StateMachine_4.AI_Locomotion.AnimStateTransitionNode_28.Transition
  nodes [Result(bCanEnterTransition=Get bStateRule_Idle_Moving() -> (bStateRule_Idle_Moving:bool))]
 }
 Transition {
  path=AnimGraph.AnimGraphNode_StateMachine_4.AI_Locomotion.AnimStateTransitionNode_3.Transition
  nodes [Result(bCanEnterTransition=Get bIsJumping())]
 }
 Transition {
  path=AnimGraph.AnimGraphNode_StateMachine_4.AI_Locomotion.AnimStateTransitionNode_31.Transition
  nodes [
   Result(bCanEnterTransition=Get bStateRule_Moving_Idle() -> (bStateRule_Moving_Idle:bool))
   AND Boolean(
    A="float < float"!(A:double=Get Speed(), B:double=10) -> (ReturnValue:bool) {is=title K2Node_CallFunction}
    B=NOT Boolean(A=Get bStartTurn() -> (bStartTurn:bool))
   )
  ]
 }
 Transition {
  path=AnimGraph.AnimGraphNode_StateMachine_4.AI_Locomotion.AnimStateTransitionNode_36.Transition
  nodes [Result(bCanEnterTransition=Get bIsSurfaceSwimming())]
 }
 Transition {
  path=AnimGraph.AnimGraphNode_StateMachine_4.AI_Locomotion.AnimStateTransitionNode_37.Transition
  nodes [Result(bCanEnterTransition=Get bIsSurfaceSwimming())]
 }
 Transition {
  path=AnimGraph.AnimGraphNode_StateMachine_4.AI_Locomotion.AnimStateTransitionNode_39.Transition
  nodes [Result(bCanEnterTransition=NOT Boolean(A=Get bIsSurfaceSwimming()))]
 }
 Transition {
  path=AnimGraph.AnimGraphNode_StateMachine_4.AI_Locomotion.AnimStateTransitionNode_4.Transition
  nodes [Result(bCanEnterTransition=Get bIsJumping())]
 }
 Transition {
  path=AnimGraph.AnimGraphNode_StateMachine_4.AI_Locomotion.AnimStateTransitionNode_41.Transition
  nodes [Result(bCanEnterTransition=Get bIsFalling())]
 }
 Transition {
  path=AnimGraph.AnimGraphNode_StateMachine_4.AI_Locomotion.AnimStateTransitionNode_43.Transition
  nodes [Result(bCanEnterTransition=Get bStateRule_Fall_Idle() -> (bStateRule_Fall_Idle:bool))]
 }
 Transition {
  path=AnimGraph.AnimGraphNode_StateMachine_4.AI_Locomotion.AnimStateTransitionNode_44.Transition
  nodes [Result(bCanEnterTransition=Get bStateRule_Land_Idle() -> (bStateRule_Land_Idle:bool))]
 }
 Transition {
  path=AnimGraph.AnimGraphNode_StateMachine_4.AI_Locomotion.AnimStateTransitionNode_45.Transition
  nodes [Result(bCanEnterTransition=Get bIsJumping())]
 }
 Transition {
  path=AnimGraph.AnimGraphNode_StateMachine_4.AI_Locomotion.AnimStateTransitionNode_47.Transition
  nodes [Result(bCanEnterTransition=Get bIsJumping())]
 }
 Transition {
  path=AnimGraph.AnimGraphNode_StateMachine_4.AI_Locomotion.AnimStateTransitionNode_48.Transition
  nodes [Result(bCanEnterTransition=Get bIsSurfaceSwimming())]
 }
 Transition {
  path=AnimGraph.AnimGraphNode_StateMachine_4.AI_Locomotion.AnimStateTransitionNode_49.Transition
  nodes [Result(bCanEnterTransition=Get bStateRule_Swimming_Fall() -> (bStateRule_Swimming_Fall:bool))]
 }
 AnimationTransitionGraph_0 {
  path=AnimGraph.AnimGraphNode_StateMachine_4.AI_Locomotion.AnimStateTransitionNode_5.AnimationTransitionGraph_0
  nodes [Result(bCanEnterTransition=Get StateTransition_FallAsleepToSleeping() -> (StateTransition_FallAsleepToSleeping:bool))]
 }
 Transition {
  path=AnimGraph.AnimGraphNode_StateMachine_4.AI_Locomotion.AnimStateTransitionNode_6.Transition
  nodes [Result(bCanEnterTransition=Get StateTransition_WakeUpToIdle() -> (StateTransition_WakeUpToIdle:bool))]
 }
 Transition {
  path=AnimGraph.AnimGraphNode_StateMachine_4.AI_Locomotion.AnimStateTransitionNode_7.Transition
  nodes [Result(bCanEnterTransition=Get bStateRule_Idle_Moving())]
 }
 Transition {path=AnimGraph.AnimGraphNode_StateMachine_4.AI_Locomotion.AnimStateTransitionNode_8.Transition, nodes [Result()]}
 RobertRidingStateMachine {
  entry=Conduit
  path=AnimGraph.AnimGraphNode_StateMachine_5.RobertRidingStateMachine
  ids=id
  nodes [
   AnimStateConduitNode_0 = Conduit()
   AnimStateNode_0 = Locomotion() {is=title AnimStateNode}
   AnimStateNode_1 = Swim() {is=title AnimStateNode}
   AnimStateNode_2 = SlopeSliding() {is=title AnimStateNode}
   AnimStateTransitionNode_0 = AnimStateTransitionNode() {crossfadeDuration=0.35, logicType=TLT_Inertialization, priorityOrder=100}
   AnimStateTransitionNode_1 = AnimStateTransitionNode() {crossfadeDuration=0.5, logicType=TLT_Inertialization}
   AnimStateTransitionNode_2 = AnimStateTransitionNode() {crossfadeDuration=0.35, logicType=TLT_Inertialization, priorityOrder=2}
   AnimStateTransitionNode_3 = AnimStateTransitionNode()
   AnimStateTransitionNode_4 = AnimStateTransitionNode()
   AnimStateTransitionNode_5 = AnimStateTransitionNode()
  ]
  links [
   {
    dst=Locomotion
    rule=AnimGraph.AnimGraphNode_StateMachine_5.RobertRidingStateMachine.AnimStateTransitionNode_0.Transition
    src=Conduit
    via=AnimStateTransitionNode_0
   }
   {
    dst=SlopeSliding
    rule=AnimGraph.AnimGraphNode_StateMachine_5.RobertRidingStateMachine.AnimStateTransitionNode_2.Transition
    src=Conduit
    via=AnimStateTransitionNode_2
   }
   {
    dst=Swim
    rule=AnimGraph.AnimGraphNode_StateMachine_5.RobertRidingStateMachine.AnimStateTransitionNode_1.Transition
    src=Conduit
    via=AnimStateTransitionNode_1
   }
   {
    dst=Conduit
    rule=AnimGraph.AnimGraphNode_StateMachine_5.RobertRidingStateMachine.AnimStateTransitionNode_5.Transition
    src=Locomotion
    via=AnimStateTransitionNode_5
   }
   {
    dst=Conduit
    rule=AnimGraph.AnimGraphNode_StateMachine_5.RobertRidingStateMachine.AnimStateTransitionNode_3.Transition
    src=SlopeSliding
    via=AnimStateTransitionNode_3
   }
   {
    dst=Conduit
    rule=AnimGraph.AnimGraphNode_StateMachine_5.RobertRidingStateMachine.AnimStateTransitionNode_4.Transition
    src=Swim
    via=AnimStateTransitionNode_4
   }
  ]
 }
 Conduit {
  path=AnimGraph.AnimGraphNode_StateMachine_5.RobertRidingStateMachine.AnimStateConduitNode_0.Conduit
  nodes [Result(bCanEnterTransition=true)]
 }
 Locomotion {
  path=AnimGraph.AnimGraphNode_StateMachine_5.RobertRidingStateMachine.AnimStateNode_0.Locomotion
  nodes [
   AnimGraphNode_StateMachine_2 = RiddenLocomotion() {is=title AnimGraphNode_StateMachine, bAllowConduitEntryStates=true}
   n1 = Apply Additive(
    Base=Apply Additive(
     Base=@AnimGraphNode_StateMachine_2
     Additive="Sequence Player (additive) 'Robert_Riding_Idle_TheRest'"() {sequence=&/Robert_Riding_Idle_TheRest}
     Alpha=Get AdditiveBodyAlpha() -> (AdditiveBodyAlpha:float)
    ) {alphaBoolBlend {blendOption=Linear}}
    Additive="Sequence Player (additive) 'Robert_Riding_Idle_Hips'"() {sequence=&/Robert_Riding_Idle_Hips}
    Alpha=Get AdditiveHipAlpha() -> (AdditiveHipAlpha:float)
   ) {alphaBoolBlend {blendOption=Linear}}
   n2 = "Sequence Player (additive) 'Robert_Riding_Idle_AllFace'"() {sequence=&/Robert_Riding_Idle_AllFace}
   n3 = Get RiderReferentialYaw() -> (RiderReferentialYaw:float)
   AnimGraphNode_StateMachine_1 = TailSecondaryState() {is=title AnimGraphNode_StateMachine}
   n4 = Apply Additive(
    Base=Apply Additive(
     Base=Apply Additive(Base=@n1, Additive=@n2) {alphaBoolBlend {blendOption=Linear}}
     Additive=Blendspace Player 'Robert_Riding_Locomotion_TurnPose_BS'!(X=@n3) -> (Pose) {blendSpace=&/Robert_Riding_Locomotion_TurnPose_BS}
     Alpha=Get BodyTurnAlpha() -> (BodyTurnAlpha:float)
    ) {alphaBoolBlend {blendOption=Linear}}
    Additive=@AnimGraphNode_StateMachine_1
    Alpha=Get LocomotionTailSecondary() -> (LocomotionTailSecondary:float)
   ) {alphaBoolBlend {blendOption=Linear}}
   AnimGraphNode_StateMachine_0 = HeadSecondaryState() {is=title AnimGraphNode_StateMachine}
   n5 = Get LocomotionHeadSecondary() -> (LocomotionHeadSecondary:float)
   n6 = Get PlayerSlopePitch() -> (PlayerSlopePitch:float)
   Output Animation Pose(Result=Apply Additive(
     Base=Apply Additive(Base=@n4, Additive=@AnimGraphNode_StateMachine_0, Alpha=@n5) {alphaBoolBlend {blendOption=Linear}}
     Additive=Blendspace Player 'Robert_Riding_Slope_ADD_BS'!(X=@n6) -> (Pose) {blendSpace=&/Robert_Riding_Slope_ADD_BS}
    ) {alphaBoolBlend {blendOption=Linear}})
  ]
 }
 HeadSecondaryState {
  entry=HoldState
  path=AnimGraph.AnimGraphNode_StateMachine_5.RobertRidingStateMachine.AnimStateNode_0.Locomotion.AnimGraphNode_StateMachine_0.HeadSecondaryState
  ids=id
  nodes [
   AnimStateNode_0 = HoldState() {is=title AnimStateNode}
   AnimStateNode_1 = HeadMovement() {is=title AnimStateNode}
   AnimStateTransitionNode_0 = AnimStateTransitionNode() {crossfadeDuration=0.5, logicType=TLT_Inertialization}
   AnimStateTransitionNode_1 = AnimStateTransitionNode() {crossfadeDuration=0.5, logicType=TLT_Inertialization}
  ]
  links [
   {
    dst=HoldState
    rule=AnimGraph.AnimGraphNode_StateMachine_5.RobertRidingStateMachine.AnimStateNode_0.Locomotion.AnimGraphNode_StateMachine_0.HeadSecondaryState.AnimStateTransitionNode_1.Transition
    src=HeadMovement
    via=AnimStateTransitionNode_1
   }
   {
    dst=HeadMovement
    rule=AnimGraph.AnimGraphNode_StateMachine_5.RobertRidingStateMachine.AnimStateNode_0.Locomotion.AnimGraphNode_StateMachine_0.HeadSecondaryState.AnimStateTransitionNode_0.Transition
    src=HoldState
    via=AnimStateTransitionNode_0
   }
  ]
 }
 HoldState {
  path=AnimGraph.AnimGraphNode_StateMachine_5.RobertRidingStateMachine.AnimStateNode_0.Locomotion.AnimGraphNode_StateMachine_0.HeadSecondaryState.AnimStateNode_0.HoldState
  nodes [Output Animation Pose(Result=Additive Identity Pose())]
 }
 HeadMovement {
  path=AnimGraph.AnimGraphNode_StateMachine_5.RobertRidingStateMachine.AnimStateNode_0.Locomotion.AnimGraphNode_StateMachine_0.HeadSecondaryState.AnimStateNode_1.HeadMovement
  nodes [Output Animation Pose(Result=Blendspace Player 'Robert_Riding_Locomotion_Head_Turn_Secondary_BS'!(X=Get!() -> (PawnDeltaRotation_Roll:float, PawnDeltaRotation_Pitch:float, PawnDeltaRotation_Yaw:float).PawnDeltaRotation_Yaw) -> (Pose) {blendSpace=&/Robert_Riding_Locomotion_Head_Turn_Secondary_BS})]
 }
 Transition {
  path=AnimGraph.AnimGraphNode_StateMachine_5.RobertRidingStateMachine.AnimStateNode_0.Locomotion.AnimGraphNode_StateMachine_0.HeadSecondaryState.AnimStateTransitionNode_0.Transition
  nodes [Result(bCanEnterTransition=Get bStateRule_IsLocomotionHoldToSecondary() -> (bStateRule_IsLocomotionHoldToSecondary:bool))]
 }
 Transition {
  path=AnimGraph.AnimGraphNode_StateMachine_5.RobertRidingStateMachine.AnimStateNode_0.Locomotion.AnimGraphNode_StateMachine_0.HeadSecondaryState.AnimStateTransitionNode_1.Transition
  nodes [Result(bCanEnterTransition=Get bStateRule_IsLocomotionSecondaryToHold() -> (bStateRule_IsLocomotionSecondaryToHold:bool))]
 }
 TailSecondaryState {
  entry=HoldState
  path=AnimGraph.AnimGraphNode_StateMachine_5.RobertRidingStateMachine.AnimStateNode_0.Locomotion.AnimGraphNode_StateMachine_1.TailSecondaryState
  ids=id
  nodes [
   AnimStateNode_0 = HoldState()
   AnimStateNode_1 = TailMovement() {is=title AnimStateNode}
   AnimStateTransitionNode_0 = AnimStateTransitionNode() {crossfadeDuration=0.5, logicType=TLT_Inertialization}
   AnimStateTransitionNode_1 = AnimStateTransitionNode() {crossfadeDuration=0.5, logicType=TLT_Inertialization}
  ]
  links [
   {
    dst=TailMovement
    rule=AnimGraph.AnimGraphNode_StateMachine_5.RobertRidingStateMachine.AnimStateNode_0.Locomotion.AnimGraphNode_StateMachine_1.TailSecondaryState.AnimStateTransitionNode_0.Transition
    src=HoldState
    via=AnimStateTransitionNode_0
   }
   {
    dst=HoldState
    rule=AnimGraph.AnimGraphNode_StateMachine_5.RobertRidingStateMachine.AnimStateNode_0.Locomotion.AnimGraphNode_StateMachine_1.TailSecondaryState.AnimStateTransitionNode_1.Transition
    src=TailMovement
    via=AnimStateTransitionNode_1
   }
  ]
 }
 HoldState {
  path=AnimGraph.AnimGraphNode_StateMachine_5.RobertRidingStateMachine.AnimStateNode_0.Locomotion.AnimGraphNode_StateMachine_1.TailSecondaryState.AnimStateNode_0.HoldState
  nodes [Output Animation Pose(Result=Additive Identity Pose())]
 }
 TailMovement {
  path=AnimGraph.AnimGraphNode_StateMachine_5.RobertRidingStateMachine.AnimStateNode_0.Locomotion.AnimGraphNode_StateMachine_1.TailSecondaryState.AnimStateNode_1.TailMovement
  nodes [Output Animation Pose(Result=Blendspace Player 'Robert_Riding_Locomotion_Tail_BS'!(X=Get().PawnDeltaRotation_Yaw) -> (Pose) {blendSpace=&/Robert_Riding_Locomotion_Tail_BS, y=1})]
 }
 Transition {
  path=AnimGraph.AnimGraphNode_StateMachine_5.RobertRidingStateMachine.AnimStateNode_0.Locomotion.AnimGraphNode_StateMachine_1.TailSecondaryState.AnimStateTransitionNode_0.Transition
  nodes [Result(bCanEnterTransition=Get bStateRule_IsLocomotionHoldToSecondary())]
 }
 Transition {
  path=AnimGraph.AnimGraphNode_StateMachine_5.RobertRidingStateMachine.AnimStateNode_0.Locomotion.AnimGraphNode_StateMachine_1.TailSecondaryState.AnimStateTransitionNode_1.Transition
  nodes [Result(bCanEnterTransition=Get bStateRule_IsLocomotionSecondaryToHold())]
 }
 RiddenLocomotion {
  entry=Conduit
  path=AnimGraph.AnimGraphNode_StateMachine_5.RobertRidingStateMachine.AnimStateNode_0.Locomotion.AnimGraphNode_StateMachine_2.RiddenLocomotion
  ids=id
  nodes [
   AnimStateConduitNode_0 = Conduit()
   AnimStateNode_0 = StopState() {is=title AnimStateNode}
   AnimStateNode_1 = LocomotionForwardState() {is=title AnimStateNode}
   AnimStateNode_2 = JumpState() {is=title AnimStateNode}
   AnimStateNode_3 = LocomotionBackwardState() {is=title AnimStateNode}
   AnimStateTransitionNode_0 = AnimStateTransitionNode() {crossfadeDuration=0.4, logicType=TLT_Inertialization, priorityOrder=100}
   AnimStateTransitionNode_1 = AnimStateTransitionNode()
   AnimStateTransitionNode_11 = AnimStateTransitionNode()
   AnimStateTransitionNode_2 = AnimStateTransitionNode() {crossfadeDuration=0.45, logicType=TLT_Inertialization}
   AnimStateTransitionNode_3 = AnimStateTransitionNode()
   AnimStateTransitionNode_4 = AnimStateTransitionNode() {crossfadeDuration=0.8, logicType=TLT_Inertialization, priorityOrder=3}
   AnimStateTransitionNode_5 = AnimStateTransitionNode() {crossfadeDuration=0.4, logicType=TLT_Inertialization, priorityOrder=2}
   AnimStateTransitionNode_6 = AnimStateTransitionNode()
  ]
  links [
   {
    dst=JumpState
    rule=AnimGraph.AnimGraphNode_StateMachine_5.RobertRidingStateMachine.AnimStateNode_0.Locomotion.AnimGraphNode_StateMachine_2.RiddenLocomotion.AnimStateTransitionNode_4.Transition
    src=Conduit
    via=AnimStateTransitionNode_4
   }
   {
    dst=LocomotionBackwardState
    rule=AnimGraph.AnimGraphNode_StateMachine_5.RobertRidingStateMachine.AnimStateNode_0.Locomotion.AnimGraphNode_StateMachine_2.RiddenLocomotion.AnimStateTransitionNode_5.Transition
    src=Conduit
    via=AnimStateTransitionNode_5
   }
   {
    dst=LocomotionForwardState
    rule=AnimGraph.AnimGraphNode_StateMachine_5.RobertRidingStateMachine.AnimStateNode_0.Locomotion.AnimGraphNode_StateMachine_2.RiddenLocomotion.AnimStateTransitionNode_2.Transition
    src=Conduit
    via=AnimStateTransitionNode_2
   }
   {
    dst=StopState
    rule=AnimGraph.AnimGraphNode_StateMachine_5.RobertRidingStateMachine.AnimStateNode_0.Locomotion.AnimGraphNode_StateMachine_2.RiddenLocomotion.AnimStateTransitionNode_0.Transition
    src=Conduit
    via=AnimStateTransitionNode_0
   }
   {
    dst=Conduit
    rule=AnimGraph.AnimGraphNode_StateMachine_5.RobertRidingStateMachine.AnimStateNode_0.Locomotion.AnimGraphNode_StateMachine_2.RiddenLocomotion.AnimStateTransitionNode_11.Transition
    src=JumpState
    via=AnimStateTransitionNode_11
   }
   {
    dst=Conduit
    rule=AnimGraph.AnimGraphNode_StateMachine_5.RobertRidingStateMachine.AnimStateNode_0.Locomotion.AnimGraphNode_StateMachine_2.RiddenLocomotion.AnimStateTransitionNode_6.Transition
    src=LocomotionBackwardState
    via=AnimStateTransitionNode_6
   }
   {
    dst=Conduit
    rule=AnimGraph.AnimGraphNode_StateMachine_5.RobertRidingStateMachine.AnimStateNode_0.Locomotion.AnimGraphNode_StateMachine_2.RiddenLocomotion.AnimStateTransitionNode_3.Transition
    src=LocomotionForwardState
    via=AnimStateTransitionNode_3
   }
   {
    dst=Conduit
    rule=AnimGraph.AnimGraphNode_StateMachine_5.RobertRidingStateMachine.AnimStateNode_0.Locomotion.AnimGraphNode_StateMachine_2.RiddenLocomotion.AnimStateTransitionNode_1.Transition
    src=StopState
    via=AnimStateTransitionNode_1
   }
  ]
 }
 Conduit {
  path=AnimGraph.AnimGraphNode_StateMachine_5.RobertRidingStateMachine.AnimStateNode_0.Locomotion.AnimGraphNode_StateMachine_2.RiddenLocomotion.AnimStateConduitNode_0.Conduit
  nodes [Result(bCanEnterTransition=true)]
 }
 StopState {
  path=AnimGraph.AnimGraphNode_StateMachine_5.RobertRidingStateMachine.AnimStateNode_0.Locomotion.AnimGraphNode_StateMachine_2.RiddenLocomotion.AnimStateNode_0.StopState
  nodes [
   AnimGraphNode_StateMachine_0 = StoppedStateMachine() {is=title AnimGraphNode_StateMachine, bAllowConduitEntryStates=true}
   Output Animation Pose(Result=@AnimGraphNode_StateMachine_0)
  ]
 }
 StoppedStateMachine {
  entry=Conduit
  path=AnimGraph.AnimGraphNode_StateMachine_5.RobertRidingStateMachine.AnimStateNode_0.Locomotion.AnimGraphNode_StateMachine_2.RiddenLocomotion.AnimStateNode_0.StopState.AnimGraphNode_StateMachine_0.StoppedStateMachine
  ids=id
  nodes [
   AnimStateConduitNode_0 = Conduit()
   AnimStateNode_0 = JumpIdleLand() {is=title AnimStateNode}
   AnimStateNode_1 = Idle() {bAlwaysResetOnEntry=true}
   AnimStateNode_2 = LocomotionToIdle() {is=title AnimStateNode}
   AnimStateNode_3 = IdleTurn() {is=title AnimStateNode, bAlwaysResetOnEntry=true}
   AnimStateNode_5 = IdleTurnToIdle() {is=title AnimStateNode}
   AnimStateTransitionNode_0 = AnimStateTransitionNode() {logicType=TLT_Inertialization, priorityOrder=2}
   AnimStateTransitionNode_1 = AnimStateTransitionNode() {logicType=TLT_Inertialization}
   AnimStateTransitionNode_11 = AnimStateTransitionNode() {crossfadeDuration=0.3, logicType=TLT_Inertialization}
   AnimStateTransitionNode_12 = AnimStateTransitionNode()
   AnimStateTransitionNode_13 = AnimStateTransitionNode() {crossfadeDuration=0.25, logicType=TLT_Inertialization}
   AnimStateTransitionNode_14 = AnimStateTransitionNode()
   AnimStateTransitionNode_15 = AnimStateTransitionNode()
   AnimStateTransitionNode_2 = AnimStateTransitionNode() {crossfadeDuration=0.6, logicType=TLT_Inertialization, priorityOrder=100}
   AnimStateTransitionNode_3 = AnimStateTransitionNode() {crossfadeDuration=0.4, logicType=TLT_Inertialization, priorityOrder=3}
   AnimStateTransitionNode_4 = AnimStateTransitionNode()
   AnimStateTransitionNode_5 = AnimStateTransitionNode()
   AnimStateTransitionNode_7 = AnimStateTransitionNode()
  ]
  links [
   {
    dst=Idle
    rule=AnimGraph.AnimGraphNode_StateMachine_5.RobertRidingStateMachine.AnimStateNode_0.Locomotion.AnimGraphNode_StateMachine_2.RiddenLocomotion.AnimStateNode_0.StopState.AnimGraphNode_StateMachine_0.StoppedStateMachine.AnimStateTransitionNode_2.Transition
    src=Conduit
    via=AnimStateTransitionNode_2
   }
   {
    dst=IdleTurn
    rule=AnimGraph.AnimGraphNode_StateMachine_5.RobertRidingStateMachine.AnimStateNode_0.Locomotion.AnimGraphNode_StateMachine_2.RiddenLocomotion.AnimStateNode_0.StopState.AnimGraphNode_StateMachine_0.StoppedStateMachine.AnimStateTransitionNode_3.Transition
    src=Conduit
    via=AnimStateTransitionNode_3
   }
   {
    dst=JumpIdleLand
    rule=AnimGraph.AnimGraphNode_StateMachine_5.RobertRidingStateMachine.AnimStateNode_0.Locomotion.AnimGraphNode_StateMachine_2.RiddenLocomotion.AnimStateNode_0.StopState.AnimGraphNode_StateMachine_0.StoppedStateMachine.AnimStateTransitionNode_1.Transition
    src=Conduit
    via=AnimStateTransitionNode_1
   }
   {
    dst=LocomotionToIdle
    rule=AnimGraph.AnimGraphNode_StateMachine_5.RobertRidingStateMachine.AnimStateNode_0.Locomotion.AnimGraphNode_StateMachine_2.RiddenLocomotion.AnimStateNode_0.StopState.AnimGraphNode_StateMachine_0.StoppedStateMachine.AnimStateTransitionNode_0.Transition
    src=Conduit
    via=AnimStateTransitionNode_0
   }
   {
    dst=Conduit
    rule=AnimGraph.AnimGraphNode_StateMachine_5.RobertRidingStateMachine.AnimStateNode_0.Locomotion.AnimGraphNode_StateMachine_2.RiddenLocomotion.AnimStateNode_0.StopState.AnimGraphNode_StateMachine_0.StoppedStateMachine.AnimStateTransitionNode_7.Transition
    src=Idle
    via=AnimStateTransitionNode_7
   }
   {
    dst=IdleTurnToIdle
    rule=AnimGraph.AnimGraphNode_StateMachine_5.RobertRidingStateMachine.AnimStateNode_0.Locomotion.AnimGraphNode_StateMachine_2.RiddenLocomotion.AnimStateNode_0.StopState.AnimGraphNode_StateMachine_0.StoppedStateMachine.AnimStateTransitionNode_11.AnimationTransitionGraph_5
    src=IdleTurn
    via=AnimStateTransitionNode_11
   }
   {
    dst=Conduit
    rule=AnimGraph.AnimGraphNode_StateMachine_5.RobertRidingStateMachine.AnimStateNode_0.Locomotion.AnimGraphNode_StateMachine_2.RiddenLocomotion.AnimStateNode_0.StopState.AnimGraphNode_StateMachine_0.StoppedStateMachine.AnimStateTransitionNode_12.Transition
    src=IdleTurnToIdle
    via=AnimStateTransitionNode_12
   }
   {
    dst=IdleTurn
    rule=AnimGraph.AnimGraphNode_StateMachine_5.RobertRidingStateMachine.AnimStateNode_0.Locomotion.AnimGraphNode_StateMachine_2.RiddenLocomotion.AnimStateNode_0.StopState.AnimGraphNode_StateMachine_0.StoppedStateMachine.AnimStateTransitionNode_13.AnimationTransitionGraph_5
    src=IdleTurnToIdle
    via=AnimStateTransitionNode_13
   }
   {
    dst=Conduit
    rule=AnimGraph.AnimGraphNode_StateMachine_5.RobertRidingStateMachine.AnimStateNode_0.Locomotion.AnimGraphNode_StateMachine_2.RiddenLocomotion.AnimStateNode_0.StopState.AnimGraphNode_StateMachine_0.StoppedStateMachine.AnimStateTransitionNode_14.Transition
    src=JumpIdleLand
    via=AnimStateTransitionNode_14
   }
   {
    dst=Conduit
    rule=AnimGraph.AnimGraphNode_StateMachine_5.RobertRidingStateMachine.AnimStateNode_0.Locomotion.AnimGraphNode_StateMachine_2.RiddenLocomotion.AnimStateNode_0.StopState.AnimGraphNode_StateMachine_0.StoppedStateMachine.AnimStateTransitionNode_4.Transition
    src=JumpIdleLand
    via=AnimStateTransitionNode_4
   }
   {
    dst=Conduit
    rule=AnimGraph.AnimGraphNode_StateMachine_5.RobertRidingStateMachine.AnimStateNode_0.Locomotion.AnimGraphNode_StateMachine_2.RiddenLocomotion.AnimStateNode_0.StopState.AnimGraphNode_StateMachine_0.StoppedStateMachine.AnimStateTransitionNode_15.Transition
    src=LocomotionToIdle
    via=AnimStateTransitionNode_15
   }
   {
    dst=Conduit
    rule=AnimGraph.AnimGraphNode_StateMachine_5.RobertRidingStateMachine.AnimStateNode_0.Locomotion.AnimGraphNode_StateMachine_2.RiddenLocomotion.AnimStateNode_0.StopState.AnimGraphNode_StateMachine_0.StoppedStateMachine.AnimStateTransitionNode_5.Transition
    src=LocomotionToIdle
    via=AnimStateTransitionNode_5
   }
  ]
 }
 Conduit {
  path=AnimGraph.AnimGraphNode_StateMachine_5.RobertRidingStateMachine.AnimStateNode_0.Locomotion.AnimGraphNode_StateMachine_2.RiddenLocomotion.AnimStateNode_0.StopState.AnimGraphNode_StateMachine_0.StoppedStateMachine.AnimStateConduitNode_0.Conduit
  nodes [Result(bCanEnterTransition=true)]
 }
 JumpIdleLand {
  path=AnimGraph.AnimGraphNode_StateMachine_5.RobertRidingStateMachine.AnimStateNode_0.Locomotion.AnimGraphNode_StateMachine_2.RiddenLocomotion.AnimStateNode_0.StopState.AnimGraphNode_StateMachine_0.StoppedStateMachine.AnimStateNode_0.JumpIdleLand
  nodes [Output Animation Pose(Result=Sequence Player 'Robert_Riding_Jump_Land'() {bLoopAnimation=false, sequence=&/Robert_Riding_Jump_Land})]
 }
 Idle {
  path=AnimGraph.AnimGraphNode_StateMachine_5.RobertRidingStateMachine.AnimStateNode_0.Locomotion.AnimGraphNode_StateMachine_2.RiddenLocomotion.AnimStateNode_0.StopState.AnimGraphNode_StateMachine_0.StoppedStateMachine.AnimStateNode_1.Idle
  nodes [Output Animation Pose(Result=Modify Curve!(
     SourcePose=Sequence Player 'Robert_Riding_Idle_Pose'() {bLoopAnimation=false, sequence=&/Robert_Riding_Idle_Pose}
     CurveValues_0=1
     CurveValues_1:float=1
     Alpha=1
    ) -> (Pose) {curveNames [AdditiveBodyAlpha, AdditiveHipAlpha], curveValues [0,0]})]
 }
 LocomotionToIdle {
  path=AnimGraph.AnimGraphNode_StateMachine_5.RobertRidingStateMachine.AnimStateNode_0.Locomotion.AnimGraphNode_StateMachine_2.RiddenLocomotion.AnimStateNode_0.StopState.AnimGraphNode_StateMachine_0.StoppedStateMachine.AnimStateNode_2.LocomotionToIdle
  nodes [
   AnimGraphNode_StateMachine_0 = Stopping States() {is=title AnimGraphNode_StateMachine, bAllowConduitEntryStates=true}
   Output Animation Pose(Result=@AnimGraphNode_StateMachine_0)
  ]
 }
 Stopping States {
  entry=Conduit
  path=AnimGraph.AnimGraphNode_StateMachine_5.RobertRidingStateMachine.AnimStateNode_0.Locomotion.AnimGraphNode_StateMachine_2.RiddenLocomotion.AnimStateNode_0.StopState.AnimGraphNode_StateMachine_0.StoppedStateMachine.AnimStateNode_2.LocomotionToIdle.AnimGraphNode_StateMachine_0.Stopping States
  ids=id
  nodes [
   AnimStateConduitNode_0 = Conduit()
   AnimStateNode_0 = SprintStopState() {is=title AnimStateNode}
   AnimStateNode_1 = WalkStopState() {is=title AnimStateNode}
   AnimStateNode_2 = WalkStopBackwardsState() {is=title AnimStateNode}
   AnimStateTransitionNode_1 = AnimStateTransitionNode() {crossfadeDuration=0.1, logicType=TLT_Inertialization, priorityOrder=100}
   AnimStateTransitionNode_2 = AnimStateTransitionNode() {logicType=TLT_Inertialization}
   AnimStateTransitionNode_3 = AnimStateTransitionNode() {crossfadeDuration=0.3, logicType=TLT_Inertialization, priorityOrder=2}
  ]
  links [
   {
    dst=SprintStopState
    rule=AnimGraph.AnimGraphNode_StateMachine_5.RobertRidingStateMachine.AnimStateNode_0.Locomotion.AnimGraphNode_StateMachine_2.RiddenLocomotion.AnimStateNode_0.StopState.AnimGraphNode_StateMachine_0.StoppedStateMachine.AnimStateNode_2.LocomotionToIdle.AnimGraphNode_StateMachine_0.Stopping States.AnimStateTransitionNode_2.Transition
    src=Conduit
    via=AnimStateTransitionNode_2
   }
   {
    dst=WalkStopBackwardsState
    rule=AnimGraph.AnimGraphNode_StateMachine_5.RobertRidingStateMachine.AnimStateNode_0.Locomotion.AnimGraphNode_StateMachine_2.RiddenLocomotion.AnimStateNode_0.StopState.AnimGraphNode_StateMachine_0.StoppedStateMachine.AnimStateNode_2.LocomotionToIdle.AnimGraphNode_StateMachine_0.Stopping States.AnimStateTransitionNode_3.Transition
    src=Conduit
    via=AnimStateTransitionNode_3
   }
   {
    dst=WalkStopState
    rule=AnimGraph.AnimGraphNode_StateMachine_5.RobertRidingStateMachine.AnimStateNode_0.Locomotion.AnimGraphNode_StateMachine_2.RiddenLocomotion.AnimStateNode_0.StopState.AnimGraphNode_StateMachine_0.StoppedStateMachine.AnimStateNode_2.LocomotionToIdle.AnimGraphNode_StateMachine_0.Stopping States.AnimStateTransitionNode_1.Transition
    src=Conduit
    via=AnimStateTransitionNode_1
   }
  ]
 }
 Conduit {
  path=AnimGraph.AnimGraphNode_StateMachine_5.RobertRidingStateMachine.AnimStateNode_0.Locomotion.AnimGraphNode_StateMachine_2.RiddenLocomotion.AnimStateNode_0.StopState.AnimGraphNode_StateMachine_0.StoppedStateMachine.AnimStateNode_2.LocomotionToIdle.AnimGraphNode_StateMachine_0.Stopping States.AnimStateConduitNode_0.Conduit
  nodes [Result(bCanEnterTransition=true)]
 }
 SprintStopState {
  path=AnimGraph.AnimGraphNode_StateMachine_5.RobertRidingStateMachine.AnimStateNode_0.Locomotion.AnimGraphNode_StateMachine_2.RiddenLocomotion.AnimStateNode_0.StopState.AnimGraphNode_StateMachine_0.StoppedStateMachine.AnimStateNode_2.LocomotionToIdle.AnimGraphNode_StateMachine_0.Stopping States.AnimStateNode_0.SprintStopState
  nodes [
   AnimGraphNode_StateMachine_0 = SprintStops() {is=title AnimGraphNode_StateMachine, bAllowConduitEntryStates=true}
   Output Animation Pose(Result=@AnimGraphNode_StateMachine_0)
  ]
 }
 SprintStops {
  entry=Conduit
  path=AnimGraph.AnimGraphNode_StateMachine_5.RobertRidingStateMachine.AnimStateNode_0.Locomotion.AnimGraphNode_StateMachine_2.RiddenLocomotion.AnimStateNode_0.StopState.AnimGraphNode_StateMachine_0.StoppedStateMachine.AnimStateNode_2.LocomotionToIdle.AnimGraphNode_StateMachine_0.Stopping States.AnimStateNode_0.SprintStopState.AnimGraphNode_StateMachine_0.SprintStops
  ids=id
  nodes [
   AnimStateConduitNode_0 = Conduit()
   AnimStateNode_0 = SprintStopFootL() {is=title AnimStateNode}
   AnimStateNode_1 = SprintStopFootR() {is=title AnimStateNode}
   AnimStateTransitionNode_0 = AnimStateTransitionNode() {logicType=TLT_Inertialization}
   AnimStateTransitionNode_1 = AnimStateTransitionNode() {logicType=TLT_Inertialization, priorityOrder=100}
  ]
  links [
   {
    dst=SprintStopFootL
    rule=AnimGraph.AnimGraphNode_StateMachine_5.RobertRidingStateMachine.AnimStateNode_0.Locomotion.AnimGraphNode_StateMachine_2.RiddenLocomotion.AnimStateNode_0.StopState.AnimGraphNode_StateMachine_0.StoppedStateMachine.AnimStateNode_2.LocomotionToIdle.AnimGraphNode_StateMachine_0.Stopping States.AnimStateNode_0.SprintStopState.AnimGraphNode_StateMachine_0.SprintStops.AnimStateTransitionNode_0.Transition
    src=Conduit
    via=AnimStateTransitionNode_0
   }
   {
    dst=SprintStopFootR
    rule=AnimGraph.AnimGraphNode_StateMachine_5.RobertRidingStateMachine.AnimStateNode_0.Locomotion.AnimGraphNode_StateMachine_2.RiddenLocomotion.AnimStateNode_0.StopState.AnimGraphNode_StateMachine_0.StoppedStateMachine.AnimStateNode_2.LocomotionToIdle.AnimGraphNode_StateMachine_0.Stopping States.AnimStateNode_0.SprintStopState.AnimGraphNode_StateMachine_0.SprintStops.AnimStateTransitionNode_1.Transition
    src=Conduit
    via=AnimStateTransitionNode_1
   }
  ]
 }
 Conduit {
  path=AnimGraph.AnimGraphNode_StateMachine_5.RobertRidingStateMachine.AnimStateNode_0.Locomotion.AnimGraphNode_StateMachine_2.RiddenLocomotion.AnimStateNode_0.StopState.AnimGraphNode_StateMachine_0.StoppedStateMachine.AnimStateNode_2.LocomotionToIdle.AnimGraphNode_StateMachine_0.Stopping States.AnimStateNode_0.SprintStopState.AnimGraphNode_StateMachine_0.SprintStops.AnimStateConduitNode_0.Conduit
  nodes [Result(bCanEnterTransition=true)]
 }
 SprintStopFootL {
  path=AnimGraph.AnimGraphNode_StateMachine_5.RobertRidingStateMachine.AnimStateNode_0.Locomotion.AnimGraphNode_StateMachine_2.RiddenLocomotion.AnimStateNode_0.StopState.AnimGraphNode_StateMachine_0.StoppedStateMachine.AnimStateNode_2.LocomotionToIdle.AnimGraphNode_StateMachine_0.Stopping States.AnimStateNode_0.SprintStopState.AnimGraphNode_StateMachine_0.SprintStops.AnimStateNode_0.SprintStopFootL
  nodes [Output Animation Pose(Result=Sequence Player 'Robert_Riding_LFoot_Run_Stop'() {bLoopAnimation=false, sequence=&/Robert_Riding_LFoot_Run_Stop})]
 }
 SprintStopFootR {
  path=AnimGraph.AnimGraphNode_StateMachine_5.RobertRidingStateMachine.AnimStateNode_0.Locomotion.AnimGraphNode_StateMachine_2.RiddenLocomotion.AnimStateNode_0.StopState.AnimGraphNode_StateMachine_0.StoppedStateMachine.AnimStateNode_2.LocomotionToIdle.AnimGraphNode_StateMachine_0.Stopping States.AnimStateNode_0.SprintStopState.AnimGraphNode_StateMachine_0.SprintStops.AnimStateNode_1.SprintStopFootR
  nodes [Output Animation Pose(Result=Sequence Player 'Robert_Riding_RFoot_Run_Stop'() {bLoopAnimation=false, sequence=&/Robert_Riding_RFoot_Run_Stop})]
 }
 Transition {
  path=AnimGraph.AnimGraphNode_StateMachine_5.RobertRidingStateMachine.AnimStateNode_0.Locomotion.AnimGraphNode_StateMachine_2.RiddenLocomotion.AnimStateNode_0.StopState.AnimGraphNode_StateMachine_0.StoppedStateMachine.AnimStateNode_2.LocomotionToIdle.AnimGraphNode_StateMachine_0.Stopping States.AnimStateNode_0.SprintStopState.AnimGraphNode_StateMachine_0.SprintStops.AnimStateTransitionNode_0.Transition
  nodes [Result(bCanEnterTransition=Get bStateRule_IsLeftFootStop() -> (bStateRule_IsLeftFootStop:bool))]
 }
 Transition {
  path=AnimGraph.AnimGraphNode_StateMachine_5.RobertRidingStateMachine.AnimStateNode_0.Locomotion.AnimGraphNode_StateMachine_2.RiddenLocomotion.AnimStateNode_0.StopState.AnimGraphNode_StateMachine_0.StoppedStateMachine.AnimStateNode_2.LocomotionToIdle.AnimGraphNode_StateMachine_0.Stopping States.AnimStateNode_0.SprintStopState.AnimGraphNode_StateMachine_0.SprintStops.AnimStateTransitionNode_1.Transition
  nodes [Result(bCanEnterTransition=true)]
 }
 WalkStopState {
  path=AnimGraph.AnimGraphNode_StateMachine_5.RobertRidingStateMachine.AnimStateNode_0.Locomotion.AnimGraphNode_StateMachine_2.RiddenLocomotion.AnimStateNode_0.StopState.AnimGraphNode_StateMachine_0.StoppedStateMachine.AnimStateNode_2.LocomotionToIdle.AnimGraphNode_StateMachine_0.Stopping States.AnimStateNode_1.WalkStopState
  nodes [
   AnimGraphNode_StateMachine_0 = WalkStops() {is=title AnimGraphNode_StateMachine, bAllowConduitEntryStates=true}
   Output Animation Pose(Result=@AnimGraphNode_StateMachine_0)
  ]
 }
 WalkStops {
  entry=Conduit
  path=AnimGraph.AnimGraphNode_StateMachine_5.RobertRidingStateMachine.AnimStateNode_0.Locomotion.AnimGraphNode_StateMachine_2.RiddenLocomotion.AnimStateNode_0.StopState.AnimGraphNode_StateMachine_0.StoppedStateMachine.AnimStateNode_2.LocomotionToIdle.AnimGraphNode_StateMachine_0.Stopping States.AnimStateNode_1.WalkStopState.AnimGraphNode_StateMachine_0.WalkStops
  ids=id
  nodes [
   AnimStateConduitNode_0 = Conduit()
   AnimStateNode_0 = WalkStopFootL() {is=title AnimStateNode}
   AnimStateNode_1 = WalkStopFootR() {is=title AnimStateNode}
   AnimStateTransitionNode_0 = AnimStateTransitionNode() {crossfadeDuration=0.1, logicType=TLT_Inertialization}
   AnimStateTransitionNode_1 = AnimStateTransitionNode() {crossfadeDuration=0.1, logicType=TLT_Inertialization, priorityOrder=100}
  ]
  links [
   {
    dst=WalkStopFootL
    rule=AnimGraph.AnimGraphNode_StateMachine_5.RobertRidingStateMachine.AnimStateNode_0.Locomotion.AnimGraphNode_StateMachine_2.RiddenLocomotion.AnimStateNode_0.StopState.AnimGraphNode_StateMachine_0.StoppedStateMachine.AnimStateNode_2.LocomotionToIdle.AnimGraphNode_StateMachine_0.Stopping States.AnimStateNode_1.WalkStopState.AnimGraphNode_StateMachine_0.WalkStops.AnimStateTransitionNode_0.Transition
    src=Conduit
    via=AnimStateTransitionNode_0
   }
   {
    dst=WalkStopFootR
    rule=AnimGraph.AnimGraphNode_StateMachine_5.RobertRidingStateMachine.AnimStateNode_0.Locomotion.AnimGraphNode_StateMachine_2.RiddenLocomotion.AnimStateNode_0.StopState.AnimGraphNode_StateMachine_0.StoppedStateMachine.AnimStateNode_2.LocomotionToIdle.AnimGraphNode_StateMachine_0.Stopping States.AnimStateNode_1.WalkStopState.AnimGraphNode_StateMachine_0.WalkStops.AnimStateTransitionNode_1.Transition
    src=Conduit
    via=AnimStateTransitionNode_1
   }
  ]
 }
 Conduit {
  path=AnimGraph.AnimGraphNode_StateMachine_5.RobertRidingStateMachine.AnimStateNode_0.Locomotion.AnimGraphNode_StateMachine_2.RiddenLocomotion.AnimStateNode_0.StopState.AnimGraphNode_StateMachine_0.StoppedStateMachine.AnimStateNode_2.LocomotionToIdle.AnimGraphNode_StateMachine_0.Stopping States.AnimStateNode_1.WalkStopState.AnimGraphNode_StateMachine_0.WalkStops.AnimStateConduitNode_0.Conduit
  nodes [Result(bCanEnterTransition=true)]
 }
 WalkStopFootL {
  path=AnimGraph.AnimGraphNode_StateMachine_5.RobertRidingStateMachine.AnimStateNode_0.Locomotion.AnimGraphNode_StateMachine_2.RiddenLocomotion.AnimStateNode_0.StopState.AnimGraphNode_StateMachine_0.StoppedStateMachine.AnimStateNode_2.LocomotionToIdle.AnimGraphNode_StateMachine_0.Stopping States.AnimStateNode_1.WalkStopState.AnimGraphNode_StateMachine_0.WalkStops.AnimStateNode_0.WalkStopFootL
  nodes [Output Animation Pose(Result=Apply Additive(
     Base=Sequence Player 'Robert_Riding_LFoot_Walk_Stop'() {bLoopAnimation=false, sequence=&/Robert_Riding_LFoot_Walk_Stop}
     Additive="Sequence Player (additive) 'Robert_Riding_LFoot_Walk_Stop_Hip'"() {bLoopAnimation=false, sequence=&/Robert_Riding_LFoot_Walk_Stop_Hip}
     Alpha=Get AdditiveHipStopAlpha() -> (AdditiveHipStopAlpha:float)
    ) {alphaBoolBlend {blendOption=Linear}})]
 }
 WalkStopFootR {
  path=AnimGraph.AnimGraphNode_StateMachine_5.RobertRidingStateMachine.AnimStateNode_0.Locomotion.AnimGraphNode_StateMachine_2.RiddenLocomotion.AnimStateNode_0.StopState.AnimGraphNode_StateMachine_0.StoppedStateMachine.AnimStateNode_2.LocomotionToIdle.AnimGraphNode_StateMachine_0.Stopping States.AnimStateNode_1.WalkStopState.AnimGraphNode_StateMachine_0.WalkStops.AnimStateNode_1.WalkStopFootR
  nodes [Output Animation Pose(Result=Apply Additive(
     Base=Sequence Player 'Robert_Riding_RFoot_Walk_Stop'() {bLoopAnimation=false, sequence=&/Robert_Riding_RFoot_Walk_Stop}
     Additive="Sequence Player (additive) 'Robert_Riding_RFoot_Walk_Stop_Hip'"() {bLoopAnimation=false, sequence=&/Robert_Riding_RFoot_Walk_Stop_Hip}
     Alpha=Get AdditiveHipStopAlpha()
    ) {alphaBoolBlend {blendOption=Linear}})]
 }
 Transition {
  path=AnimGraph.AnimGraphNode_StateMachine_5.RobertRidingStateMachine.AnimStateNode_0.Locomotion.AnimGraphNode_StateMachine_2.RiddenLocomotion.AnimStateNode_0.StopState.AnimGraphNode_StateMachine_0.StoppedStateMachine.AnimStateNode_2.LocomotionToIdle.AnimGraphNode_StateMachine_0.Stopping States.AnimStateNode_1.WalkStopState.AnimGraphNode_StateMachine_0.WalkStops.AnimStateTransitionNode_0.Transition
  nodes [Result(bCanEnterTransition=Get bStateRule_IsLeftFootStop())]
 }
 Transition {
  path=AnimGraph.AnimGraphNode_StateMachine_5.RobertRidingStateMachine.AnimStateNode_0.Locomotion.AnimGraphNode_StateMachine_2.RiddenLocomotion.AnimStateNode_0.StopState.AnimGraphNode_StateMachine_0.StoppedStateMachine.AnimStateNode_2.LocomotionToIdle.AnimGraphNode_StateMachine_0.Stopping States.AnimStateNode_1.WalkStopState.AnimGraphNode_StateMachine_0.WalkStops.AnimStateTransitionNode_1.Transition
  nodes [Result(bCanEnterTransition=true)]
 }
 WalkStopBackwardsState {
  path=AnimGraph.AnimGraphNode_StateMachine_5.RobertRidingStateMachine.AnimStateNode_0.Locomotion.AnimGraphNode_StateMachine_2.RiddenLocomotion.AnimStateNode_0.StopState.AnimGraphNode_StateMachine_0.StoppedStateMachine.AnimStateNode_2.LocomotionToIdle.AnimGraphNode_StateMachine_0.Stopping States.AnimStateNode_2.WalkStopBackwardsState
  nodes [Output Animation Pose(Result=Sequence Player 'Robert_Riding_LFoot_Walk_S_Stop'() {bLoopAnimation=false, sequence=&/Robert_Riding_LFoot_Walk_S_Stop})]
 }
 Transition {
  path=AnimGraph.AnimGraphNode_StateMachine_5.RobertRidingStateMachine.AnimStateNode_0.Locomotion.AnimGraphNode_StateMachine_2.RiddenLocomotion.AnimStateNode_0.StopState.AnimGraphNode_StateMachine_0.StoppedStateMachine.AnimStateNode_2.LocomotionToIdle.AnimGraphNode_StateMachine_0.Stopping States.AnimStateTransitionNode_1.Transition
  nodes [Result(bCanEnterTransition=true)]
 }
 Transition {
  path=AnimGraph.AnimGraphNode_StateMachine_5.RobertRidingStateMachine.AnimStateNode_0.Locomotion.AnimGraphNode_StateMachine_2.RiddenLocomotion.AnimStateNode_0.StopState.AnimGraphNode_StateMachine_0.StoppedStateMachine.AnimStateNode_2.LocomotionToIdle.AnimGraphNode_StateMachine_0.Stopping States.AnimStateTransitionNode_2.Transition
  nodes [Result(bCanEnterTransition=Get bStateRule_IsSprintStop() -> (bStateRule_IsSprintStop:bool))]
 }
 Transition {
  path=AnimGraph.AnimGraphNode_StateMachine_5.RobertRidingStateMachine.AnimStateNode_0.Locomotion.AnimGraphNode_StateMachine_2.RiddenLocomotion.AnimStateNode_0.StopState.AnimGraphNode_StateMachine_0.StoppedStateMachine.AnimStateNode_2.LocomotionToIdle.AnimGraphNode_StateMachine_0.Stopping States.AnimStateTransitionNode_3.Transition
  nodes [Result(bCanEnterTransition=Get bStateRule_IsStopBackwards() -> (bStateRule_IsStopBackwards:bool))]
 }
 IdleTurn {
  path=AnimGraph.AnimGraphNode_StateMachine_5.RobertRidingStateMachine.AnimStateNode_0.Locomotion.AnimGraphNode_StateMachine_2.RiddenLocomotion.AnimStateNode_0.StopState.AnimGraphNode_StateMachine_0.StoppedStateMachine.AnimStateNode_3.IdleTurn
  nodes [
   AnimGraphNode_StateMachine_0 = New State Machine() {is=title AnimGraphNode_StateMachine, bAllowConduitEntryStates=true}
   Output Animation Pose(Result=@AnimGraphNode_StateMachine_0)
  ]
 }
 New State Machine {
  entry=Conduit
  path=AnimGraph.AnimGraphNode_StateMachine_5.RobertRidingStateMachine.AnimStateNode_0.Locomotion.AnimGraphNode_StateMachine_2.RiddenLocomotion.AnimStateNode_0.StopState.AnimGraphNode_StateMachine_0.StoppedStateMachine.AnimStateNode_3.IdleTurn.AnimGraphNode_StateMachine_0.New State Machine
  ids=id
  nodes [
   AnimStateConduitNode_0 = Conduit()
   AnimStateNode_0 = TurnRight() {is=title AnimStateNode}
   AnimStateNode_1 = TurnLeft() {is=title AnimStateNode}
   AnimStateTransitionNode_0 = AnimStateTransitionNode() {crossfadeDuration=0.4, logicType=TLT_Inertialization}
   AnimStateTransitionNode_1 = AnimStateTransitionNode() {crossfadeDuration=0.4, logicType=TLT_Inertialization, priorityOrder=100}
  ]
  links [
   {
    dst=TurnLeft
    rule=AnimGraph.AnimGraphNode_StateMachine_5.RobertRidingStateMachine.AnimStateNode_0.Locomotion.AnimGraphNode_StateMachine_2.RiddenLocomotion.AnimStateNode_0.StopState.AnimGraphNode_StateMachine_0.StoppedStateMachine.AnimStateNode_3.IdleTurn.AnimGraphNode_StateMachine_0.New State Machine.AnimStateTransitionNode_1.AnimationTransitionGraph_0
    src=Conduit
    via=AnimStateTransitionNode_1
   }
   {
    dst=TurnRight
    rule=AnimGraph.AnimGraphNode_StateMachine_5.RobertRidingStateMachine.AnimStateNode_0.Locomotion.AnimGraphNode_StateMachine_2.RiddenLocomotion.AnimStateNode_0.StopState.AnimGraphNode_StateMachine_0.StoppedStateMachine.AnimStateNode_3.IdleTurn.AnimGraphNode_StateMachine_0.New State Machine.AnimStateTransitionNode_0.AnimationTransitionGraph_0
    src=Conduit
    via=AnimStateTransitionNode_0
   }
  ]
 }
 Conduit {
  path=AnimGraph.AnimGraphNode_StateMachine_5.RobertRidingStateMachine.AnimStateNode_0.Locomotion.AnimGraphNode_StateMachine_2.RiddenLocomotion.AnimStateNode_0.StopState.AnimGraphNode_StateMachine_0.StoppedStateMachine.AnimStateNode_3.IdleTurn.AnimGraphNode_StateMachine_0.New State Machine.AnimStateConduitNode_0.Conduit
  nodes [Result(bCanEnterTransition=true)]
 }
 TurnRight {
  path=AnimGraph.AnimGraphNode_StateMachine_5.RobertRidingStateMachine.AnimStateNode_0.Locomotion.AnimGraphNode_StateMachine_2.RiddenLocomotion.AnimStateNode_0.StopState.AnimGraphNode_StateMachine_0.StoppedStateMachine.AnimStateNode_3.IdleTurn.AnimGraphNode_StateMachine_0.New State Machine.AnimStateNode_0.TurnRight
  nodes [Output Animation Pose(Result=Sequence Player 'Robert_Riding_Turn90_Right'() {playRate=0.8, sequence=&/Robert_Riding_Turn90_Right})]
 }
 TurnLeft {
  path=AnimGraph.AnimGraphNode_StateMachine_5.RobertRidingStateMachine.AnimStateNode_0.Locomotion.AnimGraphNode_StateMachine_2.RiddenLocomotion.AnimStateNode_0.StopState.AnimGraphNode_StateMachine_0.StoppedStateMachine.AnimStateNode_3.IdleTurn.AnimGraphNode_StateMachine_0.New State Machine.AnimStateNode_1.TurnLeft
  nodes [Output Animation Pose(Result=Sequence Player 'Robert_Riding_Turn90_Left'() {playRate=0.8, sequence=&/Robert_Riding_Turn90_Left})]
 }
 AnimationTransitionGraph_0 {
  path=AnimGraph.AnimGraphNode_StateMachine_5.RobertRidingStateMachine.AnimStateNode_0.Locomotion.AnimGraphNode_StateMachine_2.RiddenLocomotion.AnimStateNode_0.StopState.AnimGraphNode_StateMachine_0.StoppedStateMachine.AnimStateNode_3.IdleTurn.AnimGraphNode_StateMachine_0.New State Machine.AnimStateTransitionNode_0.AnimationTransitionGraph_0
  nodes [Result(bCanEnterTransition=Get bStateRule_IsIdleTurnRight() -> (bStateRule_IsIdleTurnRight:bool))]
 }
 AnimationTransitionGraph_0 {
  path=AnimGraph.AnimGraphNode_StateMachine_5.RobertRidingStateMachine.AnimStateNode_0.Locomotion.AnimGraphNode_StateMachine_2.RiddenLocomotion.AnimStateNode_0.StopState.AnimGraphNode_StateMachine_0.StoppedStateMachine.AnimStateNode_3.IdleTurn.AnimGraphNode_StateMachine_0.New State Machine.AnimStateTransitionNode_1.AnimationTransitionGraph_0
  nodes [Result(bCanEnterTransition=true)]
 }
 IdleTurnToIdle {
  path=AnimGraph.AnimGraphNode_StateMachine_5.RobertRidingStateMachine.AnimStateNode_0.Locomotion.AnimGraphNode_StateMachine_2.RiddenLocomotion.AnimStateNode_0.StopState.AnimGraphNode_StateMachine_0.StoppedStateMachine.AnimStateNode_5.IdleTurnToIdle
  nodes [
   AnimGraphNode_StateMachine_0 = New State Machine() {bAllowConduitEntryStates=true}
   Output Animation Pose(Result=@AnimGraphNode_StateMachine_0)
  ]
 }
 New State Machine {
  entry=Conduit
  path=AnimGraph.AnimGraphNode_StateMachine_5.RobertRidingStateMachine.AnimStateNode_0.Locomotion.AnimGraphNode_StateMachine_2.RiddenLocomotion.AnimStateNode_0.StopState.AnimGraphNode_StateMachine_0.StoppedStateMachine.AnimStateNode_5.IdleTurnToIdle.AnimGraphNode_StateMachine_0.New State Machine
  ids=id
  nodes [
   AnimStateConduitNode_0 = Conduit()
   AnimStateNode_0 = TurnRightSettle() {is=title AnimStateNode}
   AnimStateNode_1 = TurnLeftSettle() {is=title AnimStateNode}
   AnimStateTransitionNode_0 = AnimStateTransitionNode() {crossfadeDuration=0.3, logicType=TLT_Inertialization}
   AnimStateTransitionNode_1 = AnimStateTransitionNode() {crossfadeDuration=0.3, logicType=TLT_Inertialization, priorityOrder=100}
  ]
  links [
   {
    dst=TurnLeftSettle
    rule=AnimGraph.AnimGraphNode_StateMachine_5.RobertRidingStateMachine.AnimStateNode_0.Locomotion.AnimGraphNode_StateMachine_2.RiddenLocomotion.AnimStateNode_0.StopState.AnimGraphNode_StateMachine_0.StoppedStateMachine.AnimStateNode_5.IdleTurnToIdle.AnimGraphNode_StateMachine_0.New State Machine.AnimStateTransitionNode_1.Transition
    src=Conduit
    via=AnimStateTransitionNode_1
   }
   {
    dst=TurnRightSettle
    rule=AnimGraph.AnimGraphNode_StateMachine_5.RobertRidingStateMachine.AnimStateNode_0.Locomotion.AnimGraphNode_StateMachine_2.RiddenLocomotion.AnimStateNode_0.StopState.AnimGraphNode_StateMachine_0.StoppedStateMachine.AnimStateNode_5.IdleTurnToIdle.AnimGraphNode_StateMachine_0.New State Machine.AnimStateTransitionNode_0.Transition
    src=Conduit
    via=AnimStateTransitionNode_0
   }
  ]
 }
 Conduit {
  path=AnimGraph.AnimGraphNode_StateMachine_5.RobertRidingStateMachine.AnimStateNode_0.Locomotion.AnimGraphNode_StateMachine_2.RiddenLocomotion.AnimStateNode_0.StopState.AnimGraphNode_StateMachine_0.StoppedStateMachine.AnimStateNode_5.IdleTurnToIdle.AnimGraphNode_StateMachine_0.New State Machine.AnimStateConduitNode_0.Conduit
  nodes [Result(bCanEnterTransition=true)]
 }
 TurnRightSettle {
  path=AnimGraph.AnimGraphNode_StateMachine_5.RobertRidingStateMachine.AnimStateNode_0.Locomotion.AnimGraphNode_StateMachine_2.RiddenLocomotion.AnimStateNode_0.StopState.AnimGraphNode_StateMachine_0.StoppedStateMachine.AnimStateNode_5.IdleTurnToIdle.AnimGraphNode_StateMachine_0.New State Machine.AnimStateNode_0.TurnRightSettle
  nodes [Output Animation Pose(Result=Sequence Player 'Robert_Riding_Turn90_Right_Settle'() {bLoopAnimation=false, sequence=&/Robert_Riding_Turn90_Right_Settle})]
 }
 TurnLeftSettle {
  path=AnimGraph.AnimGraphNode_StateMachine_5.RobertRidingStateMachine.AnimStateNode_0.Locomotion.AnimGraphNode_StateMachine_2.RiddenLocomotion.AnimStateNode_0.StopState.AnimGraphNode_StateMachine_0.StoppedStateMachine.AnimStateNode_5.IdleTurnToIdle.AnimGraphNode_StateMachine_0.New State Machine.AnimStateNode_1.TurnLeftSettle
  nodes [Output Animation Pose(Result=Sequence Player 'Robert_Riding_Turn90_Left_Settle'() {bLoopAnimation=false, sequence=&/Robert_Riding_Turn90_Left_Settle})]
 }
 Transition {
  path=AnimGraph.AnimGraphNode_StateMachine_5.RobertRidingStateMachine.AnimStateNode_0.Locomotion.AnimGraphNode_StateMachine_2.RiddenLocomotion.AnimStateNode_0.StopState.AnimGraphNode_StateMachine_0.StoppedStateMachine.AnimStateNode_5.IdleTurnToIdle.AnimGraphNode_StateMachine_0.New State Machine.AnimStateTransitionNode_0.Transition
  nodes [Result(bCanEnterTransition=Get bStateRule_IsTurningRight() -> (bStateRule_IsTurningRight:bool))]
 }
 Transition {
  path=AnimGraph.AnimGraphNode_StateMachine_5.RobertRidingStateMachine.AnimStateNode_0.Locomotion.AnimGraphNode_StateMachine_2.RiddenLocomotion.AnimStateNode_0.StopState.AnimGraphNode_StateMachine_0.StoppedStateMachine.AnimStateNode_5.IdleTurnToIdle.AnimGraphNode_StateMachine_0.New State Machine.AnimStateTransitionNode_1.Transition
  nodes [Result(bCanEnterTransition=true)]
 }
 Transition {
  path=AnimGraph.AnimGraphNode_StateMachine_5.RobertRidingStateMachine.AnimStateNode_0.Locomotion.AnimGraphNode_StateMachine_2.RiddenLocomotion.AnimStateNode_0.StopState.AnimGraphNode_StateMachine_0.StoppedStateMachine.AnimStateTransitionNode_0.Transition
  nodes [Result(bCanEnterTransition=Get bStateRule_IsLocomotionToIdle() -> (bStateRule_IsLocomotionToIdle:bool))]
 }
 Transition {
  path=AnimGraph.AnimGraphNode_StateMachine_5.RobertRidingStateMachine.AnimStateNode_0.Locomotion.AnimGraphNode_StateMachine_2.RiddenLocomotion.AnimStateNode_0.StopState.AnimGraphNode_StateMachine_0.StoppedStateMachine.AnimStateTransitionNode_1.Transition
  nodes [Result(bCanEnterTransition=Get bStateRule_IsJumpIdleLand() -> (bStateRule_IsJumpIdleLand:bool))]
 }
 AnimationTransitionGraph_5 {
  path=AnimGraph.AnimGraphNode_StateMachine_5.RobertRidingStateMachine.AnimStateNode_0.Locomotion.AnimGraphNode_StateMachine_2.RiddenLocomotion.AnimStateNode_0.StopState.AnimGraphNode_StateMachine_0.StoppedStateMachine.AnimStateTransitionNode_11.AnimationTransitionGraph_5
  nodes [Result(bCanEnterTransition=Get bStateRule_IsFinishedTurnInPlace() -> (bStateRule_IsFinishedTurnInPlace:bool))]
 }
 Transition {
  path=AnimGraph.AnimGraphNode_StateMachine_5.RobertRidingStateMachine.AnimStateNode_0.Locomotion.AnimGraphNode_StateMachine_2.RiddenLocomotion.AnimStateNode_0.StopState.AnimGraphNode_StateMachine_0.StoppedStateMachine.AnimStateTransitionNode_12.Transition
  nodes [Result(bCanEnterTransition=Get bStateRule_IsInterruptible() -> (bStateRule_IsInterruptible:bool))]
 }
 AnimationTransitionGraph_5 {
  path=AnimGraph.AnimGraphNode_StateMachine_5.RobertRidingStateMachine.AnimStateNode_0.Locomotion.AnimGraphNode_StateMachine_2.RiddenLocomotion.AnimStateNode_0.StopState.AnimGraphNode_StateMachine_0.StoppedStateMachine.AnimStateTransitionNode_13.AnimationTransitionGraph_5
  nodes [Result(bCanEnterTransition=Get bStateRule_IsIdleTurnToIdleTurn() -> (bStateRule_IsIdleTurnToIdleTurn:bool))]
 }
 Transition {
  path=AnimGraph.AnimGraphNode_StateMachine_5.RobertRidingStateMachine.AnimStateNode_0.Locomotion.AnimGraphNode_StateMachine_2.RiddenLocomotion.AnimStateNode_0.StopState.AnimGraphNode_StateMachine_0.StoppedStateMachine.AnimStateTransitionNode_14.Transition
  nodes [Result(bCanEnterTransition=Get bStateRule_IsTurnInPlace() -> (bStateRule_IsTurnInPlace:bool))]
 }
 Transition {
  path=AnimGraph.AnimGraphNode_StateMachine_5.RobertRidingStateMachine.AnimStateNode_0.Locomotion.AnimGraphNode_StateMachine_2.RiddenLocomotion.AnimStateNode_0.StopState.AnimGraphNode_StateMachine_0.StoppedStateMachine.AnimStateTransitionNode_15.Transition
  nodes [Result(bCanEnterTransition=Get bStateRule_IsExitLocomotionToIdle() -> (bStateRule_IsExitLocomotionToIdle:bool))]
 }
 Transition {
  path=AnimGraph.AnimGraphNode_StateMachine_5.RobertRidingStateMachine.AnimStateNode_0.Locomotion.AnimGraphNode_StateMachine_2.RiddenLocomotion.AnimStateNode_0.StopState.AnimGraphNode_StateMachine_0.StoppedStateMachine.AnimStateTransitionNode_2.Transition
  nodes [Result(bCanEnterTransition=true)]
 }
 Transition {
  path=AnimGraph.AnimGraphNode_StateMachine_5.RobertRidingStateMachine.AnimStateNode_0.Locomotion.AnimGraphNode_StateMachine_2.RiddenLocomotion.AnimStateNode_0.StopState.AnimGraphNode_StateMachine_0.StoppedStateMachine.AnimStateTransitionNode_3.Transition
  nodes [
   n1 = Get bStateRule_IsTurnInPlace()
   Result(bCanEnterTransition=@n1)
   n2 = Get!() -> (PawnRotationRate_Roll:float, PawnRotationRate_Pitch:float, PawnRotationRate_Yaw:float)
   OR Boolean(
    A=@n1
    B=OR Boolean(
     A="float > float"(A=@n2.PawnRotationRate_Yaw, B=0.2)
     B="float < float"(A=@n2.PawnRotationRate_Yaw, B=-0.2) {is=title K2Node_PromotableOperator}
    )
   )
  ]
 }
 Transition {
  path=AnimGraph.AnimGraphNode_StateMachine_5.RobertRidingStateMachine.AnimStateNode_0.Locomotion.AnimGraphNode_StateMachine_2.RiddenLocomotion.AnimStateNode_0.StopState.AnimGraphNode_StateMachine_0.StoppedStateMachine.AnimStateTransitionNode_4.Transition
  nodes [Result(bCanEnterTransition=Get bStateRule_IsInterruptible())]
 }
 Transition {
  path=AnimGraph.AnimGraphNode_StateMachine_5.RobertRidingStateMachine.AnimStateNode_0.Locomotion.AnimGraphNode_StateMachine_2.RiddenLocomotion.AnimStateNode_0.StopState.AnimGraphNode_StateMachine_0.StoppedStateMachine.AnimStateTransitionNode_5.Transition
  nodes [Result(bCanEnterTransition=Get bStateRule_IsInterruptible())]
 }
 Transition {
  path=AnimGraph.AnimGraphNode_StateMachine_5.RobertRidingStateMachine.AnimStateNode_0.Locomotion.AnimGraphNode_StateMachine_2.RiddenLocomotion.AnimStateNode_0.StopState.AnimGraphNode_StateMachine_0.StoppedStateMachine.AnimStateTransitionNode_7.Transition
  nodes [Result(bCanEnterTransition=true)]
 }
 LocomotionForwardState {
  path=AnimGraph.AnimGraphNode_StateMachine_5.RobertRidingStateMachine.AnimStateNode_0.Locomotion.AnimGraphNode_StateMachine_2.RiddenLocomotion.AnimStateNode_1.LocomotionForwardState
  nodes [
   n1 = Blendspace Player 'Robert_Riding_Boost_BodyRoll_BS'(X=Get BodyRoll() -> (BodyRoll:float), Y=Get PawnSpeed2D()) {blendSpace=&/Robert_Riding_Boost_BodyRoll_BS}
   n2 = Blendspace Player 'Robert_Riding_Locomotion_BS'!(X=Get PawnSpeed2D(), PlayRate:float=Get LocomotionPlayRate() -> (LocomotionPlayRate:float)) -> (Pose) {blendSpace=&/Robert_Riding_Locomotion_BS, groupName=Locomotion, method=SyncGroup}
   n3 = Get AdditiveLeanAlpha() -> (AdditiveLeanAlpha:float)
   Output Animation Pose(Result=Modify Curve!(SourcePose=Apply Additive(Base=@n2, Additive=@n1, Alpha=@n3) {alphaBoolBlend {blendOption=Linear}}, CurveValues_0=1, Alpha=1) -> (Pose) {curveNames [Locomotion], curveValues [0]})
  ]
 }
 JumpState {
  path=AnimGraph.AnimGraphNode_StateMachine_5.RobertRidingStateMachine.AnimStateNode_0.Locomotion.AnimGraphNode_StateMachine_2.RiddenLocomotion.AnimStateNode_2.JumpState
  nodes [
   AnimGraphNode_StateMachine_0 = New State Machine() {bAllowConduitEntryStates=true}
   Output Animation Pose(Result=Modify Curve!(SourcePose=@AnimGraphNode_StateMachine_0, CurveValues_0=0.3, CurveValues_1:float=0.05, CurveValues_2:float=1, CurveValues_3:float=1, Alpha=1) -> (Pose) {curveNames [AdditiveBodyAlpha, AdditiveHipAlpha, BothLegFK, Jumping], curveValues [0,0,0,0]})
  ]
 }
 New State Machine {
  entry=Conduit
  path=AnimGraph.AnimGraphNode_StateMachine_5.RobertRidingStateMachine.AnimStateNode_0.Locomotion.AnimGraphNode_StateMachine_2.RiddenLocomotion.AnimStateNode_2.JumpState.AnimGraphNode_StateMachine_0.New State Machine
  ids=id
  nodes [
   AnimStateConduitNode_0 = Conduit()
   AnimStateNode_0 = Jump()
   AnimStateNode_1 = FallLoop() {is=title AnimStateNode}
   AnimStateNode_2 = Fall()
   AnimStateTransitionNode_0 = AnimStateTransitionNode() {crossfadeDuration=0.7, logicType=TLT_Inertialization}
   AnimStateTransitionNode_1 = AnimStateTransitionNode() {crossfadeDuration=0.5, logicType=TLT_Inertialization}
   AnimStateTransitionNode_2 = AnimStateTransitionNode() {crossfadeDuration=0.8, logicType=TLT_Inertialization, priorityOrder=100}
   AnimStateTransitionNode_8 = AnimStateTransitionNode() {crossfadeDuration=0.1, logicType=TLT_Inertialization}
  ]
  links [
   {
    dst=Fall
    rule=AnimGraph.AnimGraphNode_StateMachine_5.RobertRidingStateMachine.AnimStateNode_0.Locomotion.AnimGraphNode_StateMachine_2.RiddenLocomotion.AnimStateNode_2.JumpState.AnimGraphNode_StateMachine_0.New State Machine.AnimStateTransitionNode_2.Transition
    src=Conduit
    via=AnimStateTransitionNode_2
   }
   {
    dst=Jump
    rule=AnimGraph.AnimGraphNode_StateMachine_5.RobertRidingStateMachine.AnimStateNode_0.Locomotion.AnimGraphNode_StateMachine_2.RiddenLocomotion.AnimStateNode_2.JumpState.AnimGraphNode_StateMachine_0.New State Machine.AnimStateTransitionNode_8.Transition
    src=Conduit
    via=AnimStateTransitionNode_8
   }
   {
    dst=FallLoop
    rule=AnimGraph.AnimGraphNode_StateMachine_5.RobertRidingStateMachine.AnimStateNode_0.Locomotion.AnimGraphNode_StateMachine_2.RiddenLocomotion.AnimStateNode_2.JumpState.AnimGraphNode_StateMachine_0.New State Machine.AnimStateTransitionNode_1.Transition
    src=Fall
    via=AnimStateTransitionNode_1
   }
   {
    dst=Fall
    rule=AnimGraph.AnimGraphNode_StateMachine_5.RobertRidingStateMachine.AnimStateNode_0.Locomotion.AnimGraphNode_StateMachine_2.RiddenLocomotion.AnimStateNode_2.JumpState.AnimGraphNode_StateMachine_0.New State Machine.AnimStateTransitionNode_0.Transition
    src=Jump
    via=AnimStateTransitionNode_0
   }
  ]
 }
 Conduit {
  path=AnimGraph.AnimGraphNode_StateMachine_5.RobertRidingStateMachine.AnimStateNode_0.Locomotion.AnimGraphNode_StateMachine_2.RiddenLocomotion.AnimStateNode_2.JumpState.AnimGraphNode_StateMachine_0.New State Machine.AnimStateConduitNode_0.Conduit
  nodes [Result(bCanEnterTransition=true)]
 }
 Jump {
  path=AnimGraph.AnimGraphNode_StateMachine_5.RobertRidingStateMachine.AnimStateNode_0.Locomotion.AnimGraphNode_StateMachine_2.RiddenLocomotion.AnimStateNode_2.JumpState.AnimGraphNode_StateMachine_0.New State Machine.AnimStateNode_0.Jump
  nodes [Output Animation Pose(Result=Sequence Player 'Robert_Riding_Jump_Start'() {bLoopAnimation=false, playRate=1.6, sequence=&/Robert_Riding_Jump_Start})]
 }
 FallLoop {
  path=AnimGraph.AnimGraphNode_StateMachine_5.RobertRidingStateMachine.AnimStateNode_0.Locomotion.AnimGraphNode_StateMachine_2.RiddenLocomotion.AnimStateNode_2.JumpState.AnimGraphNode_StateMachine_0.New State Machine.AnimStateNode_1.FallLoop
  nodes [Output Animation Pose(Result=Sequence Player 'Robert_Riding_Jump_FallLoop'() {playRate=1.4, sequence=&/Robert_Riding_Jump_FallLoop})]
 }
 Fall {
  path=AnimGraph.AnimGraphNode_StateMachine_5.RobertRidingStateMachine.AnimStateNode_0.Locomotion.AnimGraphNode_StateMachine_2.RiddenLocomotion.AnimStateNode_2.JumpState.AnimGraphNode_StateMachine_0.New State Machine.AnimStateNode_2.Fall
  nodes [Output Animation Pose(Result=Sequence Player 'Robert_Riding_Jump_Fall'() {bLoopAnimation=false, sequence=&/Robert_Riding_Jump_Fall})]
 }
 Transition {
  path=AnimGraph.AnimGraphNode_StateMachine_5.RobertRidingStateMachine.AnimStateNode_0.Locomotion.AnimGraphNode_StateMachine_2.RiddenLocomotion.AnimStateNode_2.JumpState.AnimGraphNode_StateMachine_0.New State Machine.AnimStateTransitionNode_0.Transition
  nodes [Result(bCanEnterTransition=Get bIsFalling())]
 }
 Transition {
  path=AnimGraph.AnimGraphNode_StateMachine_5.RobertRidingStateMachine.AnimStateNode_0.Locomotion.AnimGraphNode_StateMachine_2.RiddenLocomotion.AnimStateNode_2.JumpState.AnimGraphNode_StateMachine_0.New State Machine.AnimStateTransitionNode_1.Transition
  nodes [Result(bCanEnterTransition=Get bStateRule_IsInterruptible())]
 }
 Transition {
  path=AnimGraph.AnimGraphNode_StateMachine_5.RobertRidingStateMachine.AnimStateNode_0.Locomotion.AnimGraphNode_StateMachine_2.RiddenLocomotion.AnimStateNode_2.JumpState.AnimGraphNode_StateMachine_0.New State Machine.AnimStateTransitionNode_2.Transition
  nodes [Result(bCanEnterTransition=true)]
 }
 Transition {
  path=AnimGraph.AnimGraphNode_StateMachine_5.RobertRidingStateMachine.AnimStateNode_0.Locomotion.AnimGraphNode_StateMachine_2.RiddenLocomotion.AnimStateNode_2.JumpState.AnimGraphNode_StateMachine_0.New State Machine.AnimStateTransitionNode_8.Transition
  nodes [Result(bCanEnterTransition=Get bIsJumping())]
 }
 LocomotionBackwardState {
  path=AnimGraph.AnimGraphNode_StateMachine_5.RobertRidingStateMachine.AnimStateNode_0.Locomotion.AnimGraphNode_StateMachine_2.RiddenLocomotion.AnimStateNode_3.LocomotionBackwardState
  nodes [Output Animation Pose(Result=Sequence Player 'Robert_Riding_Walk_S'() {groupName=Locomotion, method=SyncGroup, sequence=&/Robert_Riding_Walk_S})]
 }
 Transition {
  path=AnimGraph.AnimGraphNode_StateMachine_5.RobertRidingStateMachine.AnimStateNode_0.Locomotion.AnimGraphNode_StateMachine_2.RiddenLocomotion.AnimStateTransitionNode_0.Transition
  nodes [Result(bCanEnterTransition=true)]
 }
 Transition {
  path=AnimGraph.AnimGraphNode_StateMachine_5.RobertRidingStateMachine.AnimStateNode_0.Locomotion.AnimGraphNode_StateMachine_2.RiddenLocomotion.AnimStateTransitionNode_1.Transition
  nodes [Result(bCanEnterTransition=true)]
 }
 Transition {
  path=AnimGraph.AnimGraphNode_StateMachine_5.RobertRidingStateMachine.AnimStateNode_0.Locomotion.AnimGraphNode_StateMachine_2.RiddenLocomotion.AnimStateTransitionNode_11.Transition
  nodes [
   Result(bCanEnterTransition=true)
   AND Boolean(A=NOT Boolean(A=Get bIsJumping()), B=NOT Boolean(A=Get bIsFalling()))
  ]
 }
 Transition {
  path=AnimGraph.AnimGraphNode_StateMachine_5.RobertRidingStateMachine.AnimStateNode_0.Locomotion.AnimGraphNode_StateMachine_2.RiddenLocomotion.AnimStateTransitionNode_2.Transition
  nodes [Result(bCanEnterTransition=Get bStateRule_IsLocomotion() -> (bStateRule_IsLocomotion:bool))]
 }
 Transition {
  path=AnimGraph.AnimGraphNode_StateMachine_5.RobertRidingStateMachine.AnimStateNode_0.Locomotion.AnimGraphNode_StateMachine_2.RiddenLocomotion.AnimStateTransitionNode_3.Transition
  nodes [Result(bCanEnterTransition=true)]
 }
 Transition {
  path=AnimGraph.AnimGraphNode_StateMachine_5.RobertRidingStateMachine.AnimStateNode_0.Locomotion.AnimGraphNode_StateMachine_2.RiddenLocomotion.AnimStateTransitionNode_4.Transition
  nodes [Result(bCanEnterTransition=Get bStateRule_IsToJumpState() -> (bStateRule_IsToJumpState:bool))]
 }
 Transition {
  path=AnimGraph.AnimGraphNode_StateMachine_5.RobertRidingStateMachine.AnimStateNode_0.Locomotion.AnimGraphNode_StateMachine_2.RiddenLocomotion.AnimStateTransitionNode_5.Transition
  nodes [Result(bCanEnterTransition=Get bStateRule_IsBackwardLocomotion() -> (bStateRule_IsBackwardLocomotion:bool))]
 }
 Transition {
  path=AnimGraph.AnimGraphNode_StateMachine_5.RobertRidingStateMachine.AnimStateNode_0.Locomotion.AnimGraphNode_StateMachine_2.RiddenLocomotion.AnimStateTransitionNode_6.Transition
  nodes [Result(bCanEnterTransition=true)]
 }
 Swim {
  path=AnimGraph.AnimGraphNode_StateMachine_5.RobertRidingStateMachine.AnimStateNode_1.Swim
  nodes [
   n1 = Apply Additive(
    Base=Apply Additive(
     Base=Sequence Player 'Robert_Riding_Swim_S_Pose'() {sequence=&/Robert_Riding_Swim_S_Pose}
     Additive="Sequence Player (additive) 'Robert_Riding_Swim_S'"() {sequence=&/Robert_Riding_Swim_S}
    ) {alphaBoolBlend {blendOption=Linear}}
    Additive=Blendspace Player 'Robert_Riding_Swimming_TurnPose_S_BS'!(X=Get SwimTurnYaw() -> (SwimTurnYaw:float)) -> (Pose) {blendSpace=&/Robert_Riding_Swimming_TurnPose_S_BS}
   ) {alphaBoolBlend {blendOption=Linear}}
   AnimGraphNode_StateMachine_2 = TailSecondaryBackwardsState() {is=title AnimGraphNode_StateMachine}
   n2 = Get()
   n3 = Get PawnSpeed2D()
   AnimGraphNode_StateMachine_0 = SwimForwardLocomotion() {is=title AnimGraphNode_StateMachine, bAllowConduitEntryStates=true}
   n4 = Apply Additive(
    Base=Apply Additive(
     Base=@AnimGraphNode_StateMachine_0
     Additive=Blendspace Player 'Robert_Rifing_Swim_Loco_BS'(X=@n3, Y=@n2.PawnRotationRate_Yaw) {blendSpace=&/Robert_Rifing_Swim_Loco_BS, groupName=SwimLoco, method=SyncGroup}
    ) {alphaBoolBlend {blendOption=Linear}}
    Additive=Blendspace Player 'Robert_Riding_Locomotion_Tail_BS'(X=Get RiderAimYaw() -> (RiderAimYaw:float)) {blendSpace=&Robert_Riding_Locomotion_Tail_BS, y=1}
    Alpha=Get SwimCameraRotationSecondary() -> (SwimCameraRotationSecondary:float)
   ) {alphaBoolBlend {blendOption=Linear}}
   n5 = Blendspace Player 'Robert_Riding_Locomotion_Head_Turn_Secondary_BS'(X=Get RiderAimYaw()) {blendSpace=&Robert_Riding_Locomotion_Head_Turn_Secondary_BS}
   n6 = Get SwimCameraRotationSecondary()
   n7 = Get SwimTurnYaw()
   AnimGraphNode_StateMachine_1 = TailSecondaryState()
   n8 = Apply Additive(
    Base=Apply Additive(
     Base=Apply Additive(Base=@n4, Additive=@n5, Alpha=@n6) {alphaBoolBlend {blendOption=Linear}}
     Additive=Blendspace Player 'Robert_Riding_Swim_TurnPose_N_BS'(X=@n7, Y=0) {blendSpace=&/Robert_Riding_Swim_TurnPose_N_BS}
     Alpha=0.75
    ) {alphaBoolBlend {blendOption=Linear}}
    Additive=@AnimGraphNode_StateMachine_1
    Alpha=Get SwimRaptorRotationSecondary() -> (SwimRaptorRotationSecondary:float)
   ) {alphaBoolBlend {blendOption=Linear}}
   AnimGraphNode_StateMachine_3 = HeadSecondaryState()
   n9 = Get SwimRaptorRotationSecondary()
   Output Animation Pose(Result=Blend Poses by bool(
     BlendPose_0=Apply Additive(Base=@n1, Additive=@AnimGraphNode_StateMachine_2, Alpha=0.8) {alphaBoolBlend {blendOption=Linear}}
     BlendPose_1=Apply Additive(Base=@n8, Additive=@AnimGraphNode_StateMachine_3, Alpha=@n9) {alphaBoolBlend {blendOption=Linear}}
     BlendTime_0=0.5
     BlendTime_1=0.5
     bActiveValue=Get bIsMovingBackwards() -> (bIsMovingBackwards:bool)
    ) {blendType=Sinusoidal, transitionType=Inertialization})
  ]
 }
 SwimForwardLocomotion {
  entry=Conduit
  path=AnimGraph.AnimGraphNode_StateMachine_5.RobertRidingStateMachine.AnimStateNode_1.Swim.AnimGraphNode_StateMachine_0.SwimForwardLocomotion
  ids=id
  nodes [
   AnimStateConduitNode_0 = Conduit()
   AnimStateNode_0 = Swim()
   AnimStateNode_3 = SwimTurn() {is=title AnimStateNode}
   AnimStateNode_4 = SwimTurnSettle() {is=title AnimStateNode}
   AnimStateTransitionNode_0 = AnimStateTransitionNode() {crossfadeDuration=0.4, logicType=TLT_Inertialization}
   AnimStateTransitionNode_1 = AnimStateTransitionNode()
   AnimStateTransitionNode_2 = AnimStateTransitionNode() {crossfadeDuration=0.5, logicType=TLT_Inertialization, priorityOrder=100}
   AnimStateTransitionNode_3 = AnimStateTransitionNode()
   AnimStateTransitionNode_4 = AnimStateTransitionNode()
   AnimStateTransitionNode_5 = AnimStateTransitionNode() {crossfadeDuration=0.3, logicType=TLT_Inertialization}
   AnimStateTransitionNode_6 = AnimStateTransitionNode() {crossfadeDuration=0.4, logicType=TLT_Inertialization}
  ]
  links [
   {
    dst=Swim
    rule=AnimGraph.AnimGraphNode_StateMachine_5.RobertRidingStateMachine.AnimStateNode_1.Swim.AnimGraphNode_StateMachine_0.SwimForwardLocomotion.AnimStateTransitionNode_2.Transition
    src=Conduit
    via=AnimStateTransitionNode_2
   }
   {
    dst=SwimTurn
    rule=AnimGraph.AnimGraphNode_StateMachine_5.RobertRidingStateMachine.AnimStateNode_1.Swim.AnimGraphNode_StateMachine_0.SwimForwardLocomotion.AnimStateTransitionNode_0.Transition
    src=Conduit
    via=AnimStateTransitionNode_0
   }
   {
    dst=Conduit
    rule=AnimGraph.AnimGraphNode_StateMachine_5.RobertRidingStateMachine.AnimStateNode_1.Swim.AnimGraphNode_StateMachine_0.SwimForwardLocomotion.AnimStateTransitionNode_3.Transition
    src=Swim
    via=AnimStateTransitionNode_3
   }
   {
    dst=SwimTurnSettle
    rule=AnimGraph.AnimGraphNode_StateMachine_5.RobertRidingStateMachine.AnimStateNode_1.Swim.AnimGraphNode_StateMachine_0.SwimForwardLocomotion.AnimStateTransitionNode_6.AnimationTransitionGraph_0
    src=SwimTurn
    via=AnimStateTransitionNode_6
   }
   {
    dst=Conduit
    rule=AnimGraph.AnimGraphNode_StateMachine_5.RobertRidingStateMachine.AnimStateNode_1.Swim.AnimGraphNode_StateMachine_0.SwimForwardLocomotion.AnimStateTransitionNode_1.Transition
    src=SwimTurnSettle
    via=AnimStateTransitionNode_1
   }
   {
    dst=Conduit
    rule=AnimGraph.AnimGraphNode_StateMachine_5.RobertRidingStateMachine.AnimStateNode_1.Swim.AnimGraphNode_StateMachine_0.SwimForwardLocomotion.AnimStateTransitionNode_4.Transition
    src=SwimTurnSettle
    via=AnimStateTransitionNode_4
   }
   {
    dst=SwimTurn
    rule=AnimGraph.AnimGraphNode_StateMachine_5.RobertRidingStateMachine.AnimStateNode_1.Swim.AnimGraphNode_StateMachine_0.SwimForwardLocomotion.AnimStateTransitionNode_5.Transition
    src=SwimTurnSettle
    via=AnimStateTransitionNode_5
   }
  ]
 }
 Conduit {
  path=AnimGraph.AnimGraphNode_StateMachine_5.RobertRidingStateMachine.AnimStateNode_1.Swim.AnimGraphNode_StateMachine_0.SwimForwardLocomotion.AnimStateConduitNode_0.Conduit
  nodes [Result(bCanEnterTransition=true)]
 }
 Swim {
  path=AnimGraph.AnimGraphNode_StateMachine_5.RobertRidingStateMachine.AnimStateNode_1.Swim.AnimGraphNode_StateMachine_0.SwimForwardLocomotion.AnimStateNode_0.Swim
  nodes [Output Animation Pose(Result="Sequence Evaluator (by time) 'Robert_Riding_Swim_Idle_Pose'"!(ExplicitTime:float=0) -> (Pose:Pose Link Structure) {is=title AnimGraphNode_SequenceEvaluator, sequence=&/Robert_Riding_Swim_Idle_Pose})]
 }
 SwimTurn {
  path=AnimGraph.AnimGraphNode_StateMachine_5.RobertRidingStateMachine.AnimStateNode_1.Swim.AnimGraphNode_StateMachine_0.SwimForwardLocomotion.AnimStateNode_3.SwimTurn
  nodes [
   AnimGraphNode_StateMachine_0 = New State Machine() {bAllowConduitEntryStates=true}
   Output Animation Pose(Result=@AnimGraphNode_StateMachine_0)
  ]
 }
 New State Machine {
  entry=Conduit
  path=AnimGraph.AnimGraphNode_StateMachine_5.RobertRidingStateMachine.AnimStateNode_1.Swim.AnimGraphNode_StateMachine_0.SwimForwardLocomotion.AnimStateNode_3.SwimTurn.AnimGraphNode_StateMachine_0.New State Machine
  ids=id
  nodes [
   AnimStateConduitNode_0 = Conduit()
   AnimStateNode_0 = SwimTurnLeft() {is=title AnimStateNode}
   AnimStateNode_1 = SwimTurnRight() {is=title AnimStateNode}
   AnimStateTransitionNode_0 = AnimStateTransitionNode() {crossfadeDuration=0.4, logicType=TLT_Inertialization, priorityOrder=100}
   AnimStateTransitionNode_1 = AnimStateTransitionNode() {crossfadeDuration=0.4, logicType=TLT_Inertialization}
  ]
  links [
   {
    dst=SwimTurnLeft
    rule=AnimGraph.AnimGraphNode_StateMachine_5.RobertRidingStateMachine.AnimStateNode_1.Swim.AnimGraphNode_StateMachine_0.SwimForwardLocomotion.AnimStateNode_3.SwimTurn.AnimGraphNode_StateMachine_0.New State Machine.AnimStateTransitionNode_0.Transition
    src=Conduit
    via=AnimStateTransitionNode_0
   }
   {
    dst=SwimTurnRight
    rule=AnimGraph.AnimGraphNode_StateMachine_5.RobertRidingStateMachine.AnimStateNode_1.Swim.AnimGraphNode_StateMachine_0.SwimForwardLocomotion.AnimStateNode_3.SwimTurn.AnimGraphNode_StateMachine_0.New State Machine.AnimStateTransitionNode_1.Transition
    src=Conduit
    via=AnimStateTransitionNode_1
   }
  ]
 }
 Conduit {
  path=AnimGraph.AnimGraphNode_StateMachine_5.RobertRidingStateMachine.AnimStateNode_1.Swim.AnimGraphNode_StateMachine_0.SwimForwardLocomotion.AnimStateNode_3.SwimTurn.AnimGraphNode_StateMachine_0.New State Machine.AnimStateConduitNode_0.Conduit
  nodes [Result(bCanEnterTransition=true)]
 }
 SwimTurnLeft {
  path=AnimGraph.AnimGraphNode_StateMachine_5.RobertRidingStateMachine.AnimStateNode_1.Swim.AnimGraphNode_StateMachine_0.SwimForwardLocomotion.AnimStateNode_3.SwimTurn.AnimGraphNode_StateMachine_0.New State Machine.AnimStateNode_0.SwimTurnLeft
  nodes [Output Animation Pose(Result=Sequence Player 'Robert_Riding_Swim_Turn90_L'() {sequence=&/Robert_Riding_Swim_Turn90_L})]
 }
 SwimTurnRight {
  path=AnimGraph.AnimGraphNode_StateMachine_5.RobertRidingStateMachine.AnimStateNode_1.Swim.AnimGraphNode_StateMachine_0.SwimForwardLocomotion.AnimStateNode_3.SwimTurn.AnimGraphNode_StateMachine_0.New State Machine.AnimStateNode_1.SwimTurnRight
  nodes [Output Animation Pose(Result=Sequence Player 'Robert_Riding_Swim_Turn90_R'() {sequence=&/Robert_Riding_Swim_Turn90_R})]
 }
 Transition {
  path=AnimGraph.AnimGraphNode_StateMachine_5.RobertRidingStateMachine.AnimStateNode_1.Swim.AnimGraphNode_StateMachine_0.SwimForwardLocomotion.AnimStateNode_3.SwimTurn.AnimGraphNode_StateMachine_0.New State Machine.AnimStateTransitionNode_0.Transition
  nodes [Result(bCanEnterTransition=true)]
 }
 Transition {
  path=AnimGraph.AnimGraphNode_StateMachine_5.RobertRidingStateMachine.AnimStateNode_1.Swim.AnimGraphNode_StateMachine_0.SwimForwardLocomotion.AnimStateNode_3.SwimTurn.AnimGraphNode_StateMachine_0.New State Machine.AnimStateTransitionNode_1.Transition
  nodes [Result(bCanEnterTransition=Get bStateRule_IsIdleTurnRight())]
 }
 SwimTurnSettle {
  path=AnimGraph.AnimGraphNode_StateMachine_5.RobertRidingStateMachine.AnimStateNode_1.Swim.AnimGraphNode_StateMachine_0.SwimForwardLocomotion.AnimStateNode_4.SwimTurnSettle
  nodes [
   AnimGraphNode_StateMachine_0 = New State Machine() {bAllowConduitEntryStates=true}
   Output Animation Pose(Result=@AnimGraphNode_StateMachine_0)
  ]
 }
 New State Machine {
  entry=Conduit
  path=AnimGraph.AnimGraphNode_StateMachine_5.RobertRidingStateMachine.AnimStateNode_1.Swim.AnimGraphNode_StateMachine_0.SwimForwardLocomotion.AnimStateNode_4.SwimTurnSettle.AnimGraphNode_StateMachine_0.New State Machine
  ids=id
  nodes [
   AnimStateConduitNode_0 = Conduit()
   AnimStateNode_0 = SwimTurnLeft()
   AnimStateNode_1 = SwimTurnRight()
   AnimStateTransitionNode_0 = AnimStateTransitionNode() {crossfadeDuration=0.4, logicType=TLT_Inertialization, priorityOrder=100}
   AnimStateTransitionNode_1 = AnimStateTransitionNode() {crossfadeDuration=0.4, logicType=TLT_Inertialization}
  ]
  links [
   {
    dst=SwimTurnLeft
    rule=AnimGraph.AnimGraphNode_StateMachine_5.RobertRidingStateMachine.AnimStateNode_1.Swim.AnimGraphNode_StateMachine_0.SwimForwardLocomotion.AnimStateNode_4.SwimTurnSettle.AnimGraphNode_StateMachine_0.New State Machine.AnimStateTransitionNode_0.Transition
    src=Conduit
    via=AnimStateTransitionNode_0
   }
   {
    dst=SwimTurnRight
    rule=AnimGraph.AnimGraphNode_StateMachine_5.RobertRidingStateMachine.AnimStateNode_1.Swim.AnimGraphNode_StateMachine_0.SwimForwardLocomotion.AnimStateNode_4.SwimTurnSettle.AnimGraphNode_StateMachine_0.New State Machine.AnimStateTransitionNode_1.Transition
    src=Conduit
    via=AnimStateTransitionNode_1
   }
  ]
 }
 Conduit {
  path=AnimGraph.AnimGraphNode_StateMachine_5.RobertRidingStateMachine.AnimStateNode_1.Swim.AnimGraphNode_StateMachine_0.SwimForwardLocomotion.AnimStateNode_4.SwimTurnSettle.AnimGraphNode_StateMachine_0.New State Machine.AnimStateConduitNode_0.Conduit
  nodes [Result(bCanEnterTransition=true)]
 }
 SwimTurnLeft {
  path=AnimGraph.AnimGraphNode_StateMachine_5.RobertRidingStateMachine.AnimStateNode_1.Swim.AnimGraphNode_StateMachine_0.SwimForwardLocomotion.AnimStateNode_4.SwimTurnSettle.AnimGraphNode_StateMachine_0.New State Machine.AnimStateNode_0.SwimTurnLeft
  nodes [Output Animation Pose(Result=Sequence Player 'Robert_Riding_Swim_Turn90_L_Settle'() {bLoopAnimation=false, sequence=&/Robert_Riding_Swim_Turn90_L_Settle})]
 }
 SwimTurnRight {
  path=AnimGraph.AnimGraphNode_StateMachine_5.RobertRidingStateMachine.AnimStateNode_1.Swim.AnimGraphNode_StateMachine_0.SwimForwardLocomotion.AnimStateNode_4.SwimTurnSettle.AnimGraphNode_StateMachine_0.New State Machine.AnimStateNode_1.SwimTurnRight
  nodes [Output Animation Pose(Result=Sequence Player 'Robert_Riding_Swim_Turn90_R_Settle'() {bLoopAnimation=false, sequence=&/Robert_Riding_Swim_Turn90_R_Settle})]
 }
 Transition {
  path=AnimGraph.AnimGraphNode_StateMachine_5.RobertRidingStateMachine.AnimStateNode_1.Swim.AnimGraphNode_StateMachine_0.SwimForwardLocomotion.AnimStateNode_4.SwimTurnSettle.AnimGraphNode_StateMachine_0.New State Machine.AnimStateTransitionNode_0.Transition
  nodes [Result(bCanEnterTransition=true)]
 }
 Transition {
  path=AnimGraph.AnimGraphNode_StateMachine_5.RobertRidingStateMachine.AnimStateNode_1.Swim.AnimGraphNode_StateMachine_0.SwimForwardLocomotion.AnimStateNode_4.SwimTurnSettle.AnimGraphNode_StateMachine_0.New State Machine.AnimStateTransitionNode_1.Transition
  nodes [Result(bCanEnterTransition=Get bStateRule_IsTurningRight())]
 }
 Transition {
  path=AnimGraph.AnimGraphNode_StateMachine_5.RobertRidingStateMachine.AnimStateNode_1.Swim.AnimGraphNode_StateMachine_0.SwimForwardLocomotion.AnimStateTransitionNode_0.Transition
  nodes [Result(bCanEnterTransition=Get bIsTurningInPlace() -> (bIsTurningInPlace:bool))]
 }
 Transition {
  path=AnimGraph.AnimGraphNode_StateMachine_5.RobertRidingStateMachine.AnimStateNode_1.Swim.AnimGraphNode_StateMachine_0.SwimForwardLocomotion.AnimStateTransitionNode_1.Transition
  nodes [Result(bCanEnterTransition=Get bStateRule_IsInterruptible())]
 }
 Transition {
  path=AnimGraph.AnimGraphNode_StateMachine_5.RobertRidingStateMachine.AnimStateNode_1.Swim.AnimGraphNode_StateMachine_0.SwimForwardLocomotion.AnimStateTransitionNode_2.Transition
  nodes [Result(bCanEnterTransition=true)]
 }
 Transition {
  path=AnimGraph.AnimGraphNode_StateMachine_5.RobertRidingStateMachine.AnimStateNode_1.Swim.AnimGraphNode_StateMachine_0.SwimForwardLocomotion.AnimStateTransitionNode_3.Transition
  nodes [Result(bCanEnterTransition=true)]
 }
 Transition {
  path=AnimGraph.AnimGraphNode_StateMachine_5.RobertRidingStateMachine.AnimStateNode_1.Swim.AnimGraphNode_StateMachine_0.SwimForwardLocomotion.AnimStateTransitionNode_4.Transition
  nodes [Result(bCanEnterTransition=Get bIsAccelerating2D())]
 }
 Transition {
  path=AnimGraph.AnimGraphNode_StateMachine_5.RobertRidingStateMachine.AnimStateNode_1.Swim.AnimGraphNode_StateMachine_0.SwimForwardLocomotion.AnimStateTransitionNode_5.Transition
  nodes [Result(bCanEnterTransition=Get bIsTurningInPlace())]
 }
 AnimationTransitionGraph_0 {
  path=AnimGraph.AnimGraphNode_StateMachine_5.RobertRidingStateMachine.AnimStateNode_1.Swim.AnimGraphNode_StateMachine_0.SwimForwardLocomotion.AnimStateTransitionNode_6.AnimationTransitionGraph_0
  nodes [Result(bCanEnterTransition=Get bStateRule_IsFinishedTurnInPlace())]
 }
 TailSecondaryState {
  entry=HoldState
  path=AnimGraph.AnimGraphNode_StateMachine_5.RobertRidingStateMachine.AnimStateNode_1.Swim.AnimGraphNode_StateMachine_1.TailSecondaryState
  ids=id
  nodes [
   AnimStateNode_0 = HoldState()
   AnimStateNode_1 = TailMovement()
   AnimStateTransitionNode_0 = AnimStateTransitionNode() {crossfadeDuration=0.5, logicType=TLT_Inertialization}
   AnimStateTransitionNode_1 = AnimStateTransitionNode() {crossfadeDuration=0.5, logicType=TLT_Inertialization}
  ]
  links [
   {
    dst=TailMovement
    rule=AnimGraph.AnimGraphNode_StateMachine_5.RobertRidingStateMachine.AnimStateNode_1.Swim.AnimGraphNode_StateMachine_1.TailSecondaryState.AnimStateTransitionNode_0.Transition
    src=HoldState
    via=AnimStateTransitionNode_0
   }
   {
    dst=HoldState
    rule=AnimGraph.AnimGraphNode_StateMachine_5.RobertRidingStateMachine.AnimStateNode_1.Swim.AnimGraphNode_StateMachine_1.TailSecondaryState.AnimStateTransitionNode_1.Transition
    src=TailMovement
    via=AnimStateTransitionNode_1
   }
  ]
 }
 HoldState {
  path=AnimGraph.AnimGraphNode_StateMachine_5.RobertRidingStateMachine.AnimStateNode_1.Swim.AnimGraphNode_StateMachine_1.TailSecondaryState.AnimStateNode_0.HoldState
  nodes [Output Animation Pose(Result=Additive Identity Pose())]
 }
 TailMovement {
  path=AnimGraph.AnimGraphNode_StateMachine_5.RobertRidingStateMachine.AnimStateNode_1.Swim.AnimGraphNode_StateMachine_1.TailSecondaryState.AnimStateNode_1.TailMovement
  nodes [Output Animation Pose(Result=Blendspace Player 'Robert_Riding_Locomotion_Tail_BS'(X=Get!() -> (PawnDeltaRotation_Roll:float, PawnDeltaRotation_Pitch:float, PawnDeltaRotation_Yaw:float).PawnDeltaRotation_Yaw) {blendSpace=&Robert_Riding_Locomotion_Tail_BS, y=1})]
 }
 Transition {
  path=AnimGraph.AnimGraphNode_StateMachine_5.RobertRidingStateMachine.AnimStateNode_1.Swim.AnimGraphNode_StateMachine_1.TailSecondaryState.AnimStateTransitionNode_0.Transition
  nodes [Result(bCanEnterTransition=Get bStateRule_IsLocomotionHoldToSecondary())]
 }
 Transition {
  path=AnimGraph.AnimGraphNode_StateMachine_5.RobertRidingStateMachine.AnimStateNode_1.Swim.AnimGraphNode_StateMachine_1.TailSecondaryState.AnimStateTransitionNode_1.Transition
  nodes [Result(bCanEnterTransition=Get bStateRule_IsLocomotionSecondaryToHold())]
 }
 TailSecondaryBackwardsState {
  entry=HoldState
  path=AnimGraph.AnimGraphNode_StateMachine_5.RobertRidingStateMachine.AnimStateNode_1.Swim.AnimGraphNode_StateMachine_2.TailSecondaryBackwardsState
  ids=id
  nodes [
   AnimStateNode_0 = HoldState()
   AnimStateNode_1 = TailMovement()
   AnimStateTransitionNode_0 = AnimStateTransitionNode() {crossfadeDuration=0.6, logicType=TLT_Inertialization}
   AnimStateTransitionNode_1 = AnimStateTransitionNode() {crossfadeDuration=0.6, logicType=TLT_Inertialization}
  ]
  links [
   {
    dst=TailMovement
    rule=AnimGraph.AnimGraphNode_StateMachine_5.RobertRidingStateMachine.AnimStateNode_1.Swim.AnimGraphNode_StateMachine_2.TailSecondaryBackwardsState.AnimStateTransitionNode_0.Transition
    src=HoldState
    via=AnimStateTransitionNode_0
   }
   {
    dst=HoldState
    rule=AnimGraph.AnimGraphNode_StateMachine_5.RobertRidingStateMachine.AnimStateNode_1.Swim.AnimGraphNode_StateMachine_2.TailSecondaryBackwardsState.AnimStateTransitionNode_1.Transition
    src=TailMovement
    via=AnimStateTransitionNode_1
   }
  ]
 }
 HoldState {
  path=AnimGraph.AnimGraphNode_StateMachine_5.RobertRidingStateMachine.AnimStateNode_1.Swim.AnimGraphNode_StateMachine_2.TailSecondaryBackwardsState.AnimStateNode_0.HoldState
  nodes [Output Animation Pose(Result=Additive Identity Pose())]
 }
 TailMovement {
  path=AnimGraph.AnimGraphNode_StateMachine_5.RobertRidingStateMachine.AnimStateNode_1.Swim.AnimGraphNode_StateMachine_2.TailSecondaryBackwardsState.AnimStateNode_1.TailMovement
  nodes [Output Animation Pose(Result=Blendspace Player 'Robert_Riding_Locomotion_Tail_BS'(X=Get().PawnDeltaRotation_Yaw) {blendSpace=&Robert_Riding_Locomotion_Tail_BS, y=1})]
 }
 Transition {
  path=AnimGraph.AnimGraphNode_StateMachine_5.RobertRidingStateMachine.AnimStateNode_1.Swim.AnimGraphNode_StateMachine_2.TailSecondaryBackwardsState.AnimStateTransitionNode_0.Transition
  nodes [Result(bCanEnterTransition=Get bStateRule_IsLocomotionHoldToSecondary())]
 }
 Transition {
  path=AnimGraph.AnimGraphNode_StateMachine_5.RobertRidingStateMachine.AnimStateNode_1.Swim.AnimGraphNode_StateMachine_2.TailSecondaryBackwardsState.AnimStateTransitionNode_1.Transition
  nodes [Result(bCanEnterTransition=Get bStateRule_IsLocomotionSecondaryToHold())]
 }
 HeadSecondaryState {
  entry=HoldState
  path=AnimGraph.AnimGraphNode_StateMachine_5.RobertRidingStateMachine.AnimStateNode_1.Swim.AnimGraphNode_StateMachine_3.HeadSecondaryState
  ids=id
  nodes [
   AnimStateNode_0 = HoldState()
   AnimStateNode_1 = HeadMovement()
   AnimStateTransitionNode_0 = AnimStateTransitionNode() {crossfadeDuration=0.5, logicType=TLT_Inertialization}
   AnimStateTransitionNode_1 = AnimStateTransitionNode() {crossfadeDuration=0.5, logicType=TLT_Inertialization}
  ]
  links [
   {
    dst=HoldState
    rule=AnimGraph.AnimGraphNode_StateMachine_5.RobertRidingStateMachine.AnimStateNode_1.Swim.AnimGraphNode_StateMachine_3.HeadSecondaryState.AnimStateTransitionNode_1.Transition
    src=HeadMovement
    via=AnimStateTransitionNode_1
   }
   {
    dst=HeadMovement
    rule=AnimGraph.AnimGraphNode_StateMachine_5.RobertRidingStateMachine.AnimStateNode_1.Swim.AnimGraphNode_StateMachine_3.HeadSecondaryState.AnimStateTransitionNode_0.Transition
    src=HoldState
    via=AnimStateTransitionNode_0
   }
  ]
 }
 HoldState {
  path=AnimGraph.AnimGraphNode_StateMachine_5.RobertRidingStateMachine.AnimStateNode_1.Swim.AnimGraphNode_StateMachine_3.HeadSecondaryState.AnimStateNode_0.HoldState
  nodes [Output Animation Pose(Result=Additive Identity Pose())]
 }
 HeadMovement {
  path=AnimGraph.AnimGraphNode_StateMachine_5.RobertRidingStateMachine.AnimStateNode_1.Swim.AnimGraphNode_StateMachine_3.HeadSecondaryState.AnimStateNode_1.HeadMovement
  nodes [Output Animation Pose(Result=Blendspace Player 'Robert_Riding_Locomotion_Head_Turn_Secondary_BS'(X=Get SwimHeadYaw() -> (SwimHeadYaw:float)) {blendSpace=&Robert_Riding_Locomotion_Head_Turn_Secondary_BS})]
 }
 Transition {
  path=AnimGraph.AnimGraphNode_StateMachine_5.RobertRidingStateMachine.AnimStateNode_1.Swim.AnimGraphNode_StateMachine_3.HeadSecondaryState.AnimStateTransitionNode_0.Transition
  nodes [Result(bCanEnterTransition=Get bStateRule_IsLocomotionHoldToSecondary())]
 }
 Transition {
  path=AnimGraph.AnimGraphNode_StateMachine_5.RobertRidingStateMachine.AnimStateNode_1.Swim.AnimGraphNode_StateMachine_3.HeadSecondaryState.AnimStateTransitionNode_1.Transition
  nodes [Result(bCanEnterTransition=Get bStateRule_IsLocomotionSecondaryToHold())]
 }
 SlopeSliding {
  path=AnimGraph.AnimGraphNode_StateMachine_5.RobertRidingStateMachine.AnimStateNode_2.SlopeSliding
  nodes [
   n1 = Apply Additive(
    Base=Sequence Player 'Robert_Riding_SlopeSlideLoop'() {sequence=&/Robert_Riding_SlopeSlideLoop}
    Additive=Blendspace Player 'Robert_Riding_SlopeSlide_Poses_BS'(X=Get SlopeSlidingRoll() -> (SlopeSlidingRoll:float), Y=Get SlopeSlidingPitch() -> (SlopeSlidingPitch:float)) {blendSpace=&/Robert_Riding_SlopeSlide_Poses_BS}
   ) {alphaBoolBlend {blendOption=Linear}}
   n2 = Blendspace Player 'Robert_Riding_Locomotion_Tail_BS'(X=Get().PawnDeltaRotation_Yaw) {blendSpace=&Robert_Riding_Locomotion_Tail_BS, y=1}
   n3 = Get()
   Output Animation Pose(Result=Apply Additive(
     Base=Apply Additive(Base=@n1, Additive=@n2, Alpha=0.6) {alphaBoolBlend {blendOption=Linear}}
     Additive=Blendspace Player 'Robert_Riding_Locomotion_Head_Turn_Secondary_BS'(X=@n3.PawnDeltaRotation_Yaw) {blendSpace=&Robert_Riding_Locomotion_Head_Turn_Secondary_BS}
     Alpha=0.6
    ) {alphaBoolBlend {blendOption=Linear}})
  ]
 }
 Transition {
  path=AnimGraph.AnimGraphNode_StateMachine_5.RobertRidingStateMachine.AnimStateTransitionNode_0.Transition
  nodes [Result(bCanEnterTransition=true)]
 }
 Transition {
  path=AnimGraph.AnimGraphNode_StateMachine_5.RobertRidingStateMachine.AnimStateTransitionNode_1.Transition
  nodes [Result(bCanEnterTransition=Get bIsSurfaceSwimming())]
 }
 Transition {
  path=AnimGraph.AnimGraphNode_StateMachine_5.RobertRidingStateMachine.AnimStateTransitionNode_2.Transition
  nodes [Result(bCanEnterTransition=Get bStateRule_IsAllowSlopeSlide() -> (bStateRule_IsAllowSlopeSlide:bool))]
 }
 Transition {
  path=AnimGraph.AnimGraphNode_StateMachine_5.RobertRidingStateMachine.AnimStateTransitionNode_3.Transition
  nodes [Result(bCanEnterTransition=true)]
 }
 Transition {
  path=AnimGraph.AnimGraphNode_StateMachine_5.RobertRidingStateMachine.AnimStateTransitionNode_4.Transition
  nodes [Result(bCanEnterTransition=true)]
 }
 Transition {
  path=AnimGraph.AnimGraphNode_StateMachine_5.RobertRidingStateMachine.AnimStateTransitionNode_5.Transition
  nodes [Result(bCanEnterTransition=true)]
 }
 BoostStatemachine {
  entry=BoostState
  path=AnimGraph.AnimGraphNode_StateMachine_6.BoostStatemachine
  ids=id
  nodes [AnimStateNode_0 = BoostState() {is=title AnimStateNode}]
 }
 BoostState {
  path=AnimGraph.AnimGraphNode_StateMachine_6.BoostStatemachine.AnimStateNode_0.BoostState
  nodes [
   AnimGraphNode_StateMachine_1 = TailSecondaryState()
   Apply Additive(Additive=@AnimGraphNode_StateMachine_1) {alphaBoolBlend {blendOption=Linear}}
   n1 = Get PawnSpeed2D()
   n2 = Get BodyRoll()
   AnimGraphNode_StateMachine_0 = New State Machine() {bAllowConduitEntryStates=true}
   n3 = Apply Additive(
    Base=Apply Additive(
     Base=@AnimGraphNode_StateMachine_0
     Additive=Blendspace Player 'Robert_Riding_Boost_BodyRoll_BS'(X=@n2, Y=@n1) {blendSpace=&Robert_Riding_Boost_BodyRoll_BS}
    ) {alphaBoolBlend {blendOption=Linear}}
    Additive=Blendspace Player 'Robert_Riding_Slope_Boost_BS'!(X=Get PlayerSlopePitch()) -> (Pose) {blendSpace=&/Robert_Riding_Slope_Boost_BS}
   ) {alphaBoolBlend {blendOption=Linear}}
   Output Animation Pose(Result=Modify Curve!(SourcePose=Modify Curve(SourcePose=@n3, CurveValues_0=1, Alpha=1) {curveNames [Sprint], curveValues [0]}, CurveValues_0=0, Alpha=1) -> (Pose) {curveNames [AdditiveAlpha], curveValues [0]})
  ]
 }
 New State Machine {
  entry=Conduit
  path=AnimGraph.AnimGraphNode_StateMachine_6.BoostStatemachine.AnimStateNode_0.BoostState.AnimGraphNode_StateMachine_0.New State Machine
  ids=id
  nodes [
   AnimStateConduitNode_0 = Conduit()
   AnimStateNode_0 = Boost() {is=title AnimStateNode}
   AnimStateNode_1 = BoostJump() {is=title AnimStateNode}
   AnimStateNode_2 = BoostFallLoop() {is=title AnimStateNode}
   AnimStateNode_3 = Boost Land() {is=title AnimStateNode}
   AnimStateTransitionNode_0 = AnimStateTransitionNode() {logicType=TLT_Inertialization, priorityOrder=100}
   AnimStateTransitionNode_1 = AnimStateTransitionNode()
   AnimStateTransitionNode_2 = AnimStateTransitionNode() {logicType=TLT_Inertialization}
   AnimStateTransitionNode_3 = AnimStateTransitionNode() {crossfadeDuration=0.5, logicType=TLT_Inertialization, priorityOrder=2}
   AnimStateTransitionNode_4 = AnimStateTransitionNode() {crossfadeDuration=0.3, logicType=TLT_Inertialization}
   AnimStateTransitionNode_5 = AnimStateTransitionNode() {crossfadeDuration=0.15, logicType=TLT_Inertialization}
   AnimStateTransitionNode_6 = AnimStateTransitionNode() {crossfadeDuration=0.1, logicType=TLT_Inertialization}
   AnimStateTransitionNode_7 = AnimStateTransitionNode() {bAutomaticRuleBasedOnSequencePlayerInState=true, crossfadeDuration=0.25}
   AnimStateTransitionNode_8 = AnimStateTransitionNode() {crossfadeDuration=0.1, logicType=TLT_Inertialization, priorityOrder=2}
  ]
  links [
   {
    dst=Conduit
    rule=AnimGraph.AnimGraphNode_StateMachine_6.BoostStatemachine.AnimStateNode_0.BoostState.AnimGraphNode_StateMachine_0.New State Machine.AnimStateTransitionNode_1.Transition
    src=Boost
    via=AnimStateTransitionNode_1
   }
   {
    dst=Boost
    rule=AnimGraph.AnimGraphNode_StateMachine_6.BoostStatemachine.AnimStateNode_0.BoostState.AnimGraphNode_StateMachine_0.New State Machine.AnimStateTransitionNode_7.Transition
    src=Boost Land
    via=AnimStateTransitionNode_7
   }
   {
    dst=BoostJump
    rule=AnimGraph.AnimGraphNode_StateMachine_6.BoostStatemachine.AnimStateNode_0.BoostState.AnimGraphNode_StateMachine_0.New State Machine.AnimStateTransitionNode_8.Transition
    src=Boost Land
    via=AnimStateTransitionNode_8
   }
   {
    dst=Boost Land
    rule=AnimGraph.AnimGraphNode_StateMachine_6.BoostStatemachine.AnimStateNode_0.BoostState.AnimGraphNode_StateMachine_0.New State Machine.AnimStateTransitionNode_6.Transition
    src=BoostFallLoop
    via=AnimStateTransitionNode_6
   }
   {
    dst=Boost Land
    rule=AnimGraph.AnimGraphNode_StateMachine_6.BoostStatemachine.AnimStateNode_0.BoostState.AnimGraphNode_StateMachine_0.New State Machine.AnimStateTransitionNode_5.Transition
    src=BoostJump
    via=AnimStateTransitionNode_5
   }
   {
    dst=BoostFallLoop
    rule=AnimGraph.AnimGraphNode_StateMachine_6.BoostStatemachine.AnimStateNode_0.BoostState.AnimGraphNode_StateMachine_0.New State Machine.AnimStateTransitionNode_4.Transition
    src=BoostJump
    via=AnimStateTransitionNode_4
   }
   {
    dst=Boost
    rule=AnimGraph.AnimGraphNode_StateMachine_6.BoostStatemachine.AnimStateNode_0.BoostState.AnimGraphNode_StateMachine_0.New State Machine.AnimStateTransitionNode_0.Transition
    src=Conduit
    via=AnimStateTransitionNode_0
   }
   {
    dst=BoostFallLoop
    rule=AnimGraph.AnimGraphNode_StateMachine_6.BoostStatemachine.AnimStateNode_0.BoostState.AnimGraphNode_StateMachine_0.New State Machine.AnimStateTransitionNode_3.Transition
    src=Conduit
    via=AnimStateTransitionNode_3
   }
   {
    dst=BoostJump
    rule=AnimGraph.AnimGraphNode_StateMachine_6.BoostStatemachine.AnimStateNode_0.BoostState.AnimGraphNode_StateMachine_0.New State Machine.AnimStateTransitionNode_2.Transition
    src=Conduit
    via=AnimStateTransitionNode_2
   }
  ]
 }
 Conduit {
  path=AnimGraph.AnimGraphNode_StateMachine_6.BoostStatemachine.AnimStateNode_0.BoostState.AnimGraphNode_StateMachine_0.New State Machine.AnimStateConduitNode_0.Conduit
  nodes [Result(bCanEnterTransition=true)]
 }
 Boost {
  path=AnimGraph.AnimGraphNode_StateMachine_6.BoostStatemachine.AnimStateNode_0.BoostState.AnimGraphNode_StateMachine_0.New State Machine.AnimStateNode_0.Boost
  nodes [Output Animation Pose(Result=Sequence Player 'Robert_Riding_Boost'() {groupName=Locomotion, method=SyncGroup, sequence=&/Robert_Riding_Boost})]
 }
 BoostJump {
  path=AnimGraph.AnimGraphNode_StateMachine_6.BoostStatemachine.AnimStateNode_0.BoostState.AnimGraphNode_StateMachine_0.New State Machine.AnimStateNode_1.BoostJump
  nodes [Output Animation Pose(Result=Sequence Player 'Robert_Riding_BoostJumpStart'!(StartPosition:float=Get BoostStartPosition() -> (BoostStartPosition:float)) -> (Pose) {bLoopAnimation=false, sequence=&/Robert_Riding_BoostJumpStart})]
 }
 BoostFallLoop {
  path=AnimGraph.AnimGraphNode_StateMachine_6.BoostStatemachine.AnimStateNode_0.BoostState.AnimGraphNode_StateMachine_0.New State Machine.AnimStateNode_2.BoostFallLoop
  nodes [Output Animation Pose(Result=Sequence Player 'Robert_Riding_BoostJumpFallLoop'() {sequence=&/Robert_Riding_BoostJumpFallLoop})]
 }
 Boost Land {
  path=AnimGraph.AnimGraphNode_StateMachine_6.BoostStatemachine.AnimStateNode_0.BoostState.AnimGraphNode_StateMachine_0.New State Machine.AnimStateNode_3.Boost Land
  nodes [Output Animation Pose(Result=Sequence Player 'Robert_Riding_BoostJumpLand'() {bLoopAnimation=false, sequence=&/Robert_Riding_BoostJumpLand})]
 }
 Transition {
  path=AnimGraph.AnimGraphNode_StateMachine_6.BoostStatemachine.AnimStateNode_0.BoostState.AnimGraphNode_StateMachine_0.New State Machine.AnimStateTransitionNode_0.Transition
  nodes [Result(bCanEnterTransition=true)]
 }
 Transition {
  path=AnimGraph.AnimGraphNode_StateMachine_6.BoostStatemachine.AnimStateNode_0.BoostState.AnimGraphNode_StateMachine_0.New State Machine.AnimStateTransitionNode_1.Transition
  nodes [Result(bCanEnterTransition=true)]
 }
 Transition {
  path=AnimGraph.AnimGraphNode_StateMachine_6.BoostStatemachine.AnimStateNode_0.BoostState.AnimGraphNode_StateMachine_0.New State Machine.AnimStateTransitionNode_2.Transition
  nodes [Result(bCanEnterTransition=Get bStateRule_IsJumpToBoostJump() -> (bStateRule_IsJumpToBoostJump:bool))]
 }
 Transition {
  path=AnimGraph.AnimGraphNode_StateMachine_6.BoostStatemachine.AnimStateNode_0.BoostState.AnimGraphNode_StateMachine_0.New State Machine.AnimStateTransitionNode_3.Transition
  nodes [Result(bCanEnterTransition=Get bIsFalling())]
 }
 Transition {
  path=AnimGraph.AnimGraphNode_StateMachine_6.BoostStatemachine.AnimStateNode_0.BoostState.AnimGraphNode_StateMachine_0.New State Machine.AnimStateTransitionNode_4.Transition
  nodes [Result(bCanEnterTransition=Get bStateRule_IsInterruptible())]
 }
 Transition {
  path=AnimGraph.AnimGraphNode_StateMachine_6.BoostStatemachine.AnimStateNode_0.BoostState.AnimGraphNode_StateMachine_0.New State Machine.AnimStateTransitionNode_5.Transition
  nodes [Result(bCanEnterTransition=Get bStateRule_BoostJumpToBoostLand() -> (bStateRule_BoostJumpToBoostLand:bool))]
 }
 Transition {
  path=AnimGraph.AnimGraphNode_StateMachine_6.BoostStatemachine.AnimStateNode_0.BoostState.AnimGraphNode_StateMachine_0.New State Machine.AnimStateTransitionNode_6.Transition
  nodes [Result(bCanEnterTransition=Get bStateRule_BoostFallLoopToBoostLand() -> (bStateRule_BoostFallLoopToBoostLand:bool))]
 }
 Transition {
  path=AnimGraph.AnimGraphNode_StateMachine_6.BoostStatemachine.AnimStateNode_0.BoostState.AnimGraphNode_StateMachine_0.New State Machine.AnimStateTransitionNode_7.Transition
  nodes [Result()]
 }
 Transition {
  path=AnimGraph.AnimGraphNode_StateMachine_6.BoostStatemachine.AnimStateNode_0.BoostState.AnimGraphNode_StateMachine_0.New State Machine.AnimStateTransitionNode_8.Transition
  nodes [Result(bCanEnterTransition=Get bIsJumping())]
 }
 TailSecondaryState {
  entry=HoldState
  path=AnimGraph.AnimGraphNode_StateMachine_6.BoostStatemachine.AnimStateNode_0.BoostState.AnimGraphNode_StateMachine_1.TailSecondaryState
  ids=id
  nodes [
   AnimStateNode_0 = HoldState()
   AnimStateNode_1 = TailMovement()
   AnimStateTransitionNode_0 = AnimStateTransitionNode() {crossfadeDuration=0.5, logicType=TLT_Inertialization}
   AnimStateTransitionNode_1 = AnimStateTransitionNode() {crossfadeDuration=0.5, logicType=TLT_Inertialization}
  ]
  links [
   {
    dst=TailMovement
    rule=AnimGraph.AnimGraphNode_StateMachine_6.BoostStatemachine.AnimStateNode_0.BoostState.AnimGraphNode_StateMachine_1.TailSecondaryState.AnimStateTransitionNode_0.Transition
    src=HoldState
    via=AnimStateTransitionNode_0
   }
   {
    dst=HoldState
    rule=AnimGraph.AnimGraphNode_StateMachine_6.BoostStatemachine.AnimStateNode_0.BoostState.AnimGraphNode_StateMachine_1.TailSecondaryState.AnimStateTransitionNode_1.Transition
    src=TailMovement
    via=AnimStateTransitionNode_1
   }
  ]
 }
 HoldState {
  path=AnimGraph.AnimGraphNode_StateMachine_6.BoostStatemachine.AnimStateNode_0.BoostState.AnimGraphNode_StateMachine_1.TailSecondaryState.AnimStateNode_0.HoldState
  nodes [Output Animation Pose(Result=Additive Identity Pose())]
 }
 TailMovement {
  path=AnimGraph.AnimGraphNode_StateMachine_6.BoostStatemachine.AnimStateNode_0.BoostState.AnimGraphNode_StateMachine_1.TailSecondaryState.AnimStateNode_1.TailMovement
  nodes [Output Animation Pose(Result=Blendspace Player 'Robert_Riding_Locomotion_Tail_BS'(X=Get().PawnDeltaRotation_Yaw) {blendSpace=&Robert_Riding_Locomotion_Tail_BS, y=1})]
 }
 Transition {
  path=AnimGraph.AnimGraphNode_StateMachine_6.BoostStatemachine.AnimStateNode_0.BoostState.AnimGraphNode_StateMachine_1.TailSecondaryState.AnimStateTransitionNode_0.Transition
  nodes [
   n1 = Get RiderAimYaw()
   Result(bCanEnterTransition=AND Boolean(A="float < float"(A=@n1, B=0.1), B="float > float"(A=@n1, B=-0.1)))
  ]
 }
 Transition {
  path=AnimGraph.AnimGraphNode_StateMachine_6.BoostStatemachine.AnimStateNode_0.BoostState.AnimGraphNode_StateMachine_1.TailSecondaryState.AnimStateTransitionNode_1.Transition
  nodes [
   n1 = Get RiderAimYaw()
   Result(bCanEnterTransition=OR Boolean(A="float < float"(A=@n1, B=-0.1), B="float > float"(A=@n1, B=0.1)))
  ]
 }
 EventGraph {nodes [
   AnimNotify_FallExit() {is=title K2Node_Event}
   > Set bStateRule_PlayLandAdditive(bStateRule_PlayLandAdditive:bool=true) -> (Output_Get:bool)
   AnimNotify_IdleExit() {is=title K2Node_Event}
   > Set bStateRule_PlayIdleToMoveAdditive(bStateRule_PlayIdleToMoveAdditive:bool=true) -> (Output_Get:bool)
   AnimNotify_LandEnter() {is=title K2Node_Event}
   > Set bStateRule_PlayLandAdditive(bStateRule_PlayLandAdditive=false)
   AnimNotify_IdleToMoveExit() {is=title K2Node_Event}
   > Set bStateRule_PlayIdleToMoveAdditive(bStateRule_PlayIdleToMoveAdditive=false)
   AnimNotify_LandAddExit() {is=title K2Node_Event}
   > Set bStateRule_PlayLandAdditive(bStateRule_PlayLandAdditive=false)
  ]}
 Fauna Idle {nodes [
   Input Pose() {name=InPose Idle}
   Output Pose(Result=Sequence Player 'Robert_Idle_Short'() {sequence=&/Robert_Idle_Short}) {layerGroup=FaunaIdleOverride}
  ]}
 Fauna IdleTurn {nodes [
   Input Pose() {name=InPose IdleTurn}
   Output Pose(Result=Blend Poses by bool(
     BlendPose_0=Sequence Player 'Robert_TurnL_90'() {bLoopAnimation=false, sequence=&/Robert_TurnL_90}
     BlendPose_1=Sequence Player 'Robert_TurnR_90'() {bLoopAnimation=false, sequence=&/Robert_TurnR_90}
     BlendTime_0=0.35
     BlendTime_1=0.35
     bActiveValue=Get bTurnDirection() -> (bTurnDirection:bool)
    ) {blendType=Linear, childUpateMode=ResetChildOnActivate, transitionType=Inertialization}) {layerGroup=FaunaIdleOverride}
  ]}
 Fauna Jump {nodes [Output Pose(Result=Sequence Player 'Robert_Abduction_short'() {sequence=&/Robert_Abduction_short}) {layerGroup=None}]}
 GetSequencerAnimSlotNames {nodes [
   GetSequencerAnimSlotNames() {is=title K2Node_FunctionEntry}
   > Return Node!(execute, ReturnValue:Array of Names=Make Array("[0]":Name=SequencerDefault) -> (Array:Array of Names))
  ]}
]
variables [
 bStateRule_PlayIdleToMoveAdditive {type=boolean}
 bStateRule_PlayLandAdditive {type=boolean}
]
