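# Sound cue for Irwin creature footsteps on mud, using the creature-steps sound class, prey step attenuation and a capped concurrency override.
# IsEnemy and IsLocal branches select one of three identically configured mixer and pitch-modulator chains, with unset values falling back to the enemy chain.
# In each chain, a quality-level node picks a random pool of fourteen mud-run waves or a six-wave subset, both excluded from branch culling.
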SoundCue /Irwin/Sounds/SurfaceFootsteps/Irwin_Mud_Footsteps_Cue
root=Branch_0
nodes [
 {id=Branch_0, children [Mixer_0,Branch_1,Mixer_0], properties {boolParameterName=IsEnemy}}
 {id=Branch_1, children [Mixer_1,Mixer_2,Mixer_0], properties {boolParameterName=IsLocal}}
 {id=Mixer_0, children [Modulator_0], properties {inputVolume [1]}}
 {id=Mixer_1, children [Modulator_1], properties {inputVolume [1]}}
 {id=Mixer_2, children [Modulator_2], properties {inputVolume [1]}}
 {id=Modulator_0, children [QualityLevel_0], properties {pitchMax=1.3, pitchMin=1}}
 {id=Modulator_1, children [QualityLevel_2], properties {pitchMax=1.3, pitchMin=1}}
 {id=Modulator_2, children [QualityLevel_3], properties {pitchMax=1.3, pitchMin=1}}
 {id=QualityLevel_0, children [Random_0,Random_0,Random_2,Random_2]}
 {id=QualityLevel_2, children [Random_5,Random_5,Random_6,Random_6]}
 {id=QualityLevel_3, children [Random_7,Random_7,Random_8,Random_8]}
 {
  id=Random_0
  children [
   WavePlayer_0
   WavePlayer_1
   WavePlayer_2
   WavePlayer_3
   WavePlayer_4
   WavePlayer_5
   WavePlayer_6
   WavePlayer_7
   WavePlayer_8
   WavePlayer_9
   WavePlayer_10
   WavePlayer_11
   WavePlayer_12
   WavePlayer_13
  ]
  properties {bShouldExcludeFromBranchCulling=true, weights [1,1,1,1,1,1,1,1,1,1,1,1,1,1]}
 }
 {
  id=Random_2
  children [WavePlayer_17, WavePlayer_18, WavePlayer_19, WavePlayer_20, WavePlayer_21, WavePlayer_22]
  properties {bShouldExcludeFromBranchCulling=true, weights [1,1,1,1,1,1]}
 }
 {
  id=Random_5
  children [
   WavePlayer_50
   WavePlayer_51
   WavePlayer_52
   WavePlayer_53
   WavePlayer_54
   WavePlayer_55
   WavePlayer_56
   WavePlayer_57
   WavePlayer_58
   WavePlayer_59
   WavePlayer_60
   WavePlayer_61
   WavePlayer_62
   WavePlayer_63
  ]
  properties {bShouldExcludeFromBranchCulling=true, weights [1,1,1,1,1,1,1,1,1,1,1,1,1,1]}
 }
 {
  id=Random_6
  children [WavePlayer_64, WavePlayer_65, WavePlayer_66, WavePlayer_67, WavePlayer_68, WavePlayer_69]
  properties {bShouldExcludeFromBranchCulling=true, weights [1,1,1,1,1,1]}
 }
 {
  id=Random_7
  children [
   WavePlayer_70
   WavePlayer_71
   WavePlayer_72
   WavePlayer_73
   WavePlayer_74
   WavePlayer_75
   WavePlayer_76
   WavePlayer_77
   WavePlayer_78
   WavePlayer_79
   WavePlayer_80
   WavePlayer_81
   WavePlayer_82
   WavePlayer_83
  ]
  properties {bShouldExcludeFromBranchCulling=true, weights [1,1,1,1,1,1,1,1,1,1,1,1,1,1]}
 }
 {
  id=Random_8
  children [WavePlayer_84, WavePlayer_85, WavePlayer_86, WavePlayer_87, WavePlayer_88, WavePlayer_89]
  properties {bShouldExcludeFromBranchCulling=true, weights [1,1,1,1,1,1]}
 }
 {id=WavePlayer_0, properties {soundWaveAssetPtr=&/Step_Mud_Run_Irwin_01}}
 {id=WavePlayer_1, properties {soundWaveAssetPtr=&/Step_Mud_Run_Irwin_02}}
 {id=WavePlayer_10, properties {soundWaveAssetPtr=&/Step_Mud_Run_Irwin_11}}
 {id=WavePlayer_11, properties {soundWaveAssetPtr=&/Step_Mud_Run_Irwin_12}}
 {id=WavePlayer_12, properties {soundWaveAssetPtr=&/Step_Mud_Run_Irwin_13}}
 {id=WavePlayer_13, properties {soundWaveAssetPtr=&/Step_Mud_Run_Irwin_14}}
 {id=WavePlayer_17, properties {soundWaveAssetPtr=&Step_Mud_Run_Irwin_01}}
 {id=WavePlayer_18, properties {soundWaveAssetPtr=&Step_Mud_Run_Irwin_02}}
 {id=WavePlayer_19, properties {soundWaveAssetPtr=&/Step_Mud_Run_Irwin_03}}
 {id=WavePlayer_2, properties {soundWaveAssetPtr=&Step_Mud_Run_Irwin_03}}
 {id=WavePlayer_20, properties {soundWaveAssetPtr=&/Step_Mud_Run_Irwin_04}}
 {id=WavePlayer_21, properties {soundWaveAssetPtr=&/Step_Mud_Run_Irwin_05}}
 {id=WavePlayer_22, properties {soundWaveAssetPtr=&/Step_Mud_Run_Irwin_06}}
 {id=WavePlayer_3, properties {soundWaveAssetPtr=&Step_Mud_Run_Irwin_04}}
 {id=WavePlayer_4, properties {soundWaveAssetPtr=&Step_Mud_Run_Irwin_05}}
 {id=WavePlayer_5, properties {soundWaveAssetPtr=&Step_Mud_Run_Irwin_06}}
 {id=WavePlayer_50, properties {soundWaveAssetPtr=&Step_Mud_Run_Irwin_01}}
 {id=WavePlayer_51, properties {soundWaveAssetPtr=&Step_Mud_Run_Irwin_02}}
 {id=WavePlayer_52, properties {soundWaveAssetPtr=&Step_Mud_Run_Irwin_03}}
 {id=WavePlayer_53, properties {soundWaveAssetPtr=&Step_Mud_Run_Irwin_04}}
 {id=WavePlayer_54, properties {soundWaveAssetPtr=&Step_Mud_Run_Irwin_05}}
 {id=WavePlayer_55, properties {soundWaveAssetPtr=&Step_Mud_Run_Irwin_06}}
 {id=WavePlayer_56, properties {soundWaveAssetPtr=&/Step_Mud_Run_Irwin_07}}
 {id=WavePlayer_57, properties {soundWaveAssetPtr=&/Step_Mud_Run_Irwin_08}}
 {id=WavePlayer_58, properties {soundWaveAssetPtr=&/Step_Mud_Run_Irwin_09}}
 {id=WavePlayer_59, properties {soundWaveAssetPtr=&/Step_Mud_Run_Irwin_10}}
 {id=WavePlayer_6, properties {soundWaveAssetPtr=&Step_Mud_Run_Irwin_07}}
 {id=WavePlayer_60, properties {soundWaveAssetPtr=&Step_Mud_Run_Irwin_11}}
 {id=WavePlayer_61, properties {soundWaveAssetPtr=&Step_Mud_Run_Irwin_12}}
 {id=WavePlayer_62, properties {soundWaveAssetPtr=&Step_Mud_Run_Irwin_13}}
 {id=WavePlayer_63, properties {soundWaveAssetPtr=&Step_Mud_Run_Irwin_14}}
 {id=WavePlayer_64, properties {soundWaveAssetPtr=&Step_Mud_Run_Irwin_01}}
 {id=WavePlayer_65, properties {soundWaveAssetPtr=&Step_Mud_Run_Irwin_02}}
 {id=WavePlayer_66, properties {soundWaveAssetPtr=&Step_Mud_Run_Irwin_03}}
 {id=WavePlayer_67, properties {soundWaveAssetPtr=&Step_Mud_Run_Irwin_04}}
 {id=WavePlayer_68, properties {soundWaveAssetPtr=&Step_Mud_Run_Irwin_05}}
 {id=WavePlayer_69, properties {soundWaveAssetPtr=&Step_Mud_Run_Irwin_06}}
 {id=WavePlayer_7, properties {soundWaveAssetPtr=&Step_Mud_Run_Irwin_08}}
 {id=WavePlayer_70, properties {soundWaveAssetPtr=&Step_Mud_Run_Irwin_01}}
 {id=WavePlayer_71, properties {soundWaveAssetPtr=&Step_Mud_Run_Irwin_02}}
 {id=WavePlayer_72, properties {soundWaveAssetPtr=&Step_Mud_Run_Irwin_03}}
 {id=WavePlayer_73, properties {soundWaveAssetPtr=&Step_Mud_Run_Irwin_04}}
 {id=WavePlayer_74, properties {soundWaveAssetPtr=&Step_Mud_Run_Irwin_05}}
 {id=WavePlayer_75, properties {soundWaveAssetPtr=&Step_Mud_Run_Irwin_06}}
 {id=WavePlayer_76, properties {soundWaveAssetPtr=&Step_Mud_Run_Irwin_07}}
 {id=WavePlayer_77, properties {soundWaveAssetPtr=&Step_Mud_Run_Irwin_08}}
 {id=WavePlayer_78, properties {soundWaveAssetPtr=&Step_Mud_Run_Irwin_09}}
 {id=WavePlayer_79, properties {soundWaveAssetPtr=&Step_Mud_Run_Irwin_10}}
 {id=WavePlayer_8, properties {soundWaveAssetPtr=&Step_Mud_Run_Irwin_09}}
 {id=WavePlayer_80, properties {soundWaveAssetPtr=&Step_Mud_Run_Irwin_11}}
 {id=WavePlayer_81, properties {soundWaveAssetPtr=&Step_Mud_Run_Irwin_12}}
 {id=WavePlayer_82, properties {soundWaveAssetPtr=&Step_Mud_Run_Irwin_13}}
 {id=WavePlayer_83, properties {soundWaveAssetPtr=&Step_Mud_Run_Irwin_14}}
 {id=WavePlayer_84, properties {soundWaveAssetPtr=&Step_Mud_Run_Irwin_01}}
 {id=WavePlayer_85, properties {soundWaveAssetPtr=&Step_Mud_Run_Irwin_02}}
 {id=WavePlayer_86, properties {soundWaveAssetPtr=&Step_Mud_Run_Irwin_03}}
 {id=WavePlayer_87, properties {soundWaveAssetPtr=&Step_Mud_Run_Irwin_04}}
 {id=WavePlayer_88, properties {soundWaveAssetPtr=&Step_Mud_Run_Irwin_05}}
 {id=WavePlayer_89, properties {soundWaveAssetPtr=&Step_Mud_Run_Irwin_06}}
 {id=WavePlayer_9, properties {soundWaveAssetPtr=&Step_Mud_Run_Irwin_10}}
]
properties {
 attenuationSettings=/Irwin/Sounds/Settings/Prey_Step_Att
 bOverrideConcurrency=true
 maxDistance=3200
 soundClassObject=/Game/Sounds/FortSoundClassesAndModes/Fort_Creatures_Steps_Tactical
 volumeMultiplier=0.5
 concurrencyOverrides {maxCount=4, voiceStealReleaseTime=0.1, volumeScaleAttackTime=0.15}
}
