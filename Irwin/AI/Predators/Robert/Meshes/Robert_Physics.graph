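# Physics asset for the Robert predator AI character, giving its skeletal mesh a coarse collision shape.
# Twelve bones each carry a single capsule: head, chest, spine, both elbows, both knees and ankles, and three spaced tail bones.
# Left and right limb capsules mirror each other, and the tail capsules narrow toward the tip.
# No constraints are defined, so the bodies are not jointed for ragdoll simulation.
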PhysicsAsset /Irwin/AI/Predators/Robert/Meshes/Robert_Physics
bodies {
 BipedArm_A_Elbow_L [{shapeType=Capsule, length=27.97, radius=10.21, center=<10.52 -6.4 -2.8>, rotation=<rot 0 -96.39 -82.77>}]
 BipedArm_A_Elbow_R [{shapeType=Capsule, length=27.97, radius=10.21, center=<-10.52 6.4 2.8>, rotation=<rot 0 -96.39 97.23>}]
 BipedHeadNeck_A_Head_C [{shapeType=Capsule, length=25, radius=22, center=<14.73 9.59 0.07>, rotation=<rot 0 168.16 -89.54>}]
 HoofedLeg_A_Ankle_L [{shapeType=Capsule, length=30, radius=16, center=<15.7 7.25 -2.73>, rotation=<rot -72.03 -57.04 60.73>}]
 HoofedLeg_A_Ankle_R [{shapeType=Capsule, length=30, radius=16, center=<-15.7 -7.25 2.73>, rotation=<rot -72.03 -57.04 -119.27>}]
 HoofedLeg_A_Knee_L [{shapeType=Capsule, length=35, radius=15, center=<15.84 0.58 -0.05>, rotation=<rot -29.15 87.99 -88.18>}]
 HoofedLeg_A_Knee_R [{shapeType=Capsule, length=35, radius=15, center=<-15.84 -0.58 0.05>, rotation=<rot -29.15 87.99 91.82>}]
 QuadSpine_A_Chest_C [{shapeType=Capsule, length=42.09, radius=20.4, center=<3.75 -0.69 0.22>, rotation=<rot -0.14 19.44 -89.6>}]
 QuadSpine_A_Spine2_C [{shapeType=Capsule, length=40, radius=30, center=<-14.08 2.17 0>, rotation=<rot 90 159.44 139.44>}]
 Tail_A_Tail1_C [{shapeType=Capsule, length=40, radius=16, rotation=<rot 90 0 0>}]
 Tail_A_Tail4_C [{shapeType=Capsule, length=40, radius=13, rotation=<rot 90 0 0>}]
 Tail_A_Tail7_C [{shapeType=Capsule, length=50, radius=10, rotation=<rot 90 0 0>}]
}
constraints []
